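# Control Rig for the M_MED_Orbit_Teal player body's cloth asset, adding corrective deformation and secondary motion to bones.
# Limb joint angles, remapped to clamped weights, blend elbow, knee, shoulder, pectoral and thigh deform bones toward sculpted poses.
# Setup spawns two dynamics solvers: right-arm tubes with soft cross-constraints, and a holster with an added end bone and box confiner.
# Each update applies correctives, parent-constrains the holster bottom, then steps both solvers.
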ControlRigBlueprint /BRCosmetics/Characters/Player/Male/Medium/Bodies/M_MED_Orbit_Teal/Meshes/ClothAsset/M_MED_Orbit_Teal_ClothEditor_Deform_CtrlRig_sculpters
graphs [
 CollapseNode_ContainedGraph {nodes [
   RigVMFunctionEntryNode()
   RigUnit_GetRelativeTransformForItem = &RigUnit_GetRelativeTransformForItem(Child=(Type=Bone,Name="lowerarm_l"), Parent=(Type=Bone,Name="upperarm_l"), bChildInitial=False, bParentInitial=False)
   n1 = &RigVMFunction_MathQuaternionToEuler(RotationOrder=ZYX, Value=@RigUnit_GetRelativeTransformForItem.RelativeTransform.Rotation)
   n2 = &RigVMFunction_MathFloatRemap(SourceMaximum=92, SourceMinimum=28, TargetMinimum=0, bClamp=true, Value=@n1.Result.Z)
   n3 = &RigVMFunction_MathQuaternionToEuler(
    RotationOrder=ZYX
    Value=&RigUnit_GetTransform(Item=(Type=Bone,Name="deform_elbow_bk_l"), Space=LocalSpace, bInitial=true).Transform.Rotation
   )
   > &RigUnit_ModifyTransforms(
    ItemToModify=((Item=(Type=Bone,Name="deform_elbow_bk_l"),Transform=(Rotation=(X=0,Y=0,Z=0,W=1),Translation=(X=-2.858,Y=-2.932,Z=-0.666),Scale3D=(X=1,Y=1,Z=1))))
    Mode=OverrideLocal
    WeightMinimum=0
    Weight=@n2.Result
    ItemToModify.0.Transform.Rotation=&RigVMFunction_MathQuaternionFromEuler(Euler=/Script/RigVM.RigVMFunction_MathVectorAdd(B=(X=0,Y=0,Z=0), A=@n3.Result) {is=struct RigVMUnitNode, template="Add::Execute(in A,in B,out Result)"}.Result).Result
   )
   n4 = &RigVMFunction_MathQuaternionToEuler(
    RotationOrder=ZYX
    Value=&RigUnit_GetTransform(Item=(Type=Bone,Name="deform_elbow_fr_l"), Space=LocalSpace, bInitial=true).Transform.Rotation
   )
   > RigUnit_ModifyTransforms = &RigUnit_ModifyTransforms(
    ItemToModify=((Item=(Type=Bone,Name="deform_elbow_fr_l"),Transform=(Rotation=(X=0,Y=0,Z=0,W=1),Translation=(X=24.635,Y=2.261,Z=-0.637),Scale3D=(X=1,Y=1,Z=1.125))))
    Mode=OverrideLocal
    WeightMinimum=0
    Weight=@n2.Result
    ItemToModify.0.Transform.Rotation=&RigVMFunction_MathQuaternionFromEuler(Euler=&RigVMFunction_MathVectorAdd(B=(X=0,Y=0,Z=0), A=@n4.Result) {template="Add::Execute(in A,in B,out Result)"}.Result).Result
   )
   n5 = &RigVMFunction_MathFloatRemap(SourceMaximum=148, SourceMinimum=91, TargetMinimum=0, bClamp=true, Value=@n1.Result.Z)
   > &RigUnit_ModifyTransforms(
    ItemToModify=((Item=(Type=Bone,Name="deform_elbow_bk_l"),Transform=(Rotation=(X=0,Y=0,Z=0,W=1),Translation=(X=-3.59,Y=-2.934,Z=-0.716),Scale3D=(X=1,Y=1,Z=1))))
    Mode=OverrideLocal
    WeightMinimum=0
    Weight=@n5.Result
    ItemToModify.0.Transform.Rotation=&RigVMFunction_MathQuaternionFromEuler(Euler=&RigVMFunction_MathVectorAdd(B=(X=0,Y=0,Z=-13.6), A=@n3.Result) {template="Add::Execute(in A,in B,out Result)"}.Result).Result
   )
   > &RigUnit_ModifyTransforms(
    ItemToModify=((Item=(Type=Bone,Name="deform_elbow_fr_l"),Transform=(Rotation=(X=0,Y=0,Z=0,W=1),Translation=(X=15.651,Y=1.255,Z=-0.922),Scale3D=(X=1,Y=1,Z=1.337))))
    Mode=OverrideLocal
    WeightMinimum=0
    Weight=@n5.Result
    ItemToModify.0.Transform.Rotation=&RigVMFunction_MathQuaternionFromEuler(Euler=&RigVMFunction_MathVectorAdd(B=(X=0,Y=0,Z=28), A=@n4.Result) {template="Add::Execute(in A,in B,out Result)"}.Result).Result
   )
   n6 = &RigVMFunction_MathQuaternionToEuler(
    RotationOrder=ZYX
    Value=&RigUnit_GetRelativeTransformForItem(Child=(Type=Bone,Name="lowerarm_r"), Parent=(Type=Bone,Name="upperarm_r"), bChildInitial=False, bParentInitial=False).RelativeTransform.Rotation
   )
   n7 = &RigVMFunction_MathFloatRemap(SourceMaximum=92, SourceMinimum=28, TargetMinimum=0, bClamp=true, Value=@n6.Result.Z)
   n8 = &RigVMFunction_MathQuaternionToEuler(
    RotationOrder=ZYX
    Value=&RigUnit_GetTransform(Item=(Type=Bone,Name="deform_elbow_bk_r"), Space=LocalSpace, bInitial=true).Transform.Rotation
   )
   > &RigUnit_ModifyTransforms(
    ItemToModify=((Item=(Type=Bone,Name="deform_elbow_bk_r"),Transform=(Rotation=(X=0,Y=0,Z=0,W=1),Translation=(X=2.858,Y=2.932,Z=0.666),Scale3D=(X=1,Y=1,Z=1))))
    Mode=OverrideLocal
    WeightMinimum=0
    Weight=@n7.Result
    ItemToModify.0.Transform.Rotation=&RigVMFunction_MathQuaternionFromEuler(Euler=&RigVMFunction_MathVectorAdd(B=(X=0,Y=0,Z=0), A=@n8.Result) {template="Add::Execute(in A,in B,out Result)"}.Result).Result
   )
   n9 = &RigVMFunction_MathQuaternionToEuler(
    RotationOrder=ZYX
    Value=&RigUnit_GetTransform(Item=(Type=Bone,Name="deform_elbow_fr_r"), Space=LocalSpace, bInitial=true).Transform.Rotation
   )
   > &RigUnit_ModifyTransforms(
    ItemToModify=((Item=(Type=Bone,Name="deform_elbow_fr_r"),Transform=(Rotation=(X=0,Y=0,Z=0,W=1),Translation=(X=-24.635,Y=-2.261,Z=0.637),Scale3D=(X=1,Y=1,Z=1.125))))
    Mode=OverrideLocal
    WeightMinimum=0
    Weight=@n7.Result
    ItemToModify.0.Transform.Rotation=&RigVMFunction_MathQuaternionFromEuler(Euler=&RigVMFunction_MathVectorAdd(B=(X=0,Y=0,Z=0), A=@n9.Result) {template="Add::Execute(in A,in B,out Result)"}.Result).Result
   )
   n10 = &RigVMFunction_MathFloatRemap(SourceMaximum=148, SourceMinimum=91, TargetMinimum=0, bClamp=true, Value=@n6.Result.Z)
   > &RigUnit_ModifyTransforms(
    ItemToModify=((Item=(Type=Bone,Name="deform_elbow_bk_r"),Transform=(Rotation=(X=0,Y=0,Z=0,W=1),Translation=(X=3.59,Y=2.934,Z=0.716),Scale3D=(X=1,Y=1,Z=1))))
    Mode=OverrideLocal
    WeightMinimum=0
    Weight=@n10.Result
    ItemToModify.0.Transform.Rotation=&RigVMFunction_MathQuaternionFromEuler(Euler=&RigVMFunction_MathVectorAdd(B=(X=0,Y=0,Z=-13.6), A=@n8.Result) {template="Add::Execute(in A,in B,out Result)"}.Result).Result
   )
   > &RigUnit_ModifyTransforms(
    ItemToModify=((Item=(Type=Bone,Name="deform_elbow_fr_r"),Transform=(Rotation=(X=0,Y=0,Z=0,W=1),Translation=(X=-15.651,Y=-1.255,Z=0.922),Scale3D=(X=1,Y=1,Z=1.337))))
    Mode=OverrideLocal
    WeightMinimum=0
    Weight=@n10.Result
    ItemToModify.0.Transform.Rotation=&RigVMFunction_MathQuaternionFromEuler(Euler=&RigVMFunction_MathVectorAdd(B=(X=0,Y=0,Z=28), A=@n9.Result) {template="Add::Execute(in A,in B,out Result)"}.Result).Result
   )
   n11 = &RigVMFunction_MathQuaternionToEuler(
    RotationOrder=ZYX
    Value=&RigUnit_GetRelativeTransformForItem(Child=(Type=Bone,Name="calf_l"), Parent=(Type=Bone,Name="thigh_l"), bChildInitial=False, bParentInitial=False).RelativeTransform.Rotation
   )
   n12 = &RigVMFunction_MathFloatRemap(SourceMaximum=114, SourceMinimum=4, TargetMinimum=0, bClamp=true, Value=@n11.Result.Z)
   n13 = &RigVMFunction_MathQuaternionToEuler(
    RotationOrder=ZYX
    Value=&RigUnit_GetTransform(Item=(Type=Bone,Name="deform_knee_bk_01_l"), Space=LocalSpace, bInitial=true).Transform.Rotation
   )
   > &RigUnit_ModifyTransforms(
    ItemToModify=((Item=(Type=Bone,Name="deform_knee_bk_01_l"),Transform=(Rotation=(X=0,Y=0,Z=0,W=1),Translation=(X=-4.319,Y=-4.798,Z=0),Scale3D=(X=1,Y=1,Z=0.7))))
    Mode=OverrideLocal
    WeightMinimum=0
    Weight=@n12.Result
    ItemToModify.0.Transform.Rotation=&RigVMFunction_MathQuaternionFromEuler(Euler=&RigVMFunction_MathVectorAdd(B=(X=0,Y=0,Z=0), A=@n13.Result) {template="Add::Execute(in A,in B,out Result)"}.Result).Result
   )
   n14 = &RigVMFunction_MathQuaternionToEuler(
    RotationOrder=ZYX
    Value=&RigUnit_GetTransform(Item=(Type=Bone,Name="deform_knee_fr_01_l"), Space=LocalSpace, bInitial=true).Transform.Rotation
   )
   > &RigUnit_ModifyTransforms(
    ItemToModify=((Item=(Type=Bone,Name="deform_knee_fr_01_l"),Transform=(Rotation=(X=0,Y=0,Z=0,W=1),Translation=(X=0.258,Y=-2,Z=0.009),Scale3D=(X=1,Y=1,Z=1))))
    Mode=OverrideLocal
    WeightMinimum=0
    Weight=@n12.Result
    ItemToModify.0.Transform.Rotation=&RigVMFunction_MathQuaternionFromEuler(Euler=&RigVMFunction_MathVectorAdd(B=(X=0,Y=0,Z=-15), A=@n14.Result) {template="Add::Execute(in A,in B,out Result)"}.Result).Result
   )
   n15 = &RigVMFunction_MathFloatRemap(SourceMaximum=130, SourceMinimum=113, TargetMinimum=0, bClamp=true, Value=@n11.Result.Z)
   > &RigUnit_ModifyTransforms(
    ItemToModify=((Item=(Type=Bone,Name="deform_knee_bk_01_l"),Transform=(Rotation=(X=0,Y=0,Z=0,W=1),Translation=(X=-10.633,Y=-3.561,Z=0),Scale3D=(X=1,Y=1,Z=1))))
    Mode=OverrideLocal
    WeightMinimum=0
    Weight=@n15.Result
    ItemToModify.0.Transform.Rotation=&RigVMFunction_MathQuaternionFromEuler(Euler=&RigVMFunction_MathVectorAdd(B=(X=0,Y=0,Z=0), A=@n13.Result) {template="Add::Execute(in A,in B,out Result)"}.Result).Result
   )
   > &RigUnit_ModifyTransforms(
    ItemToModify=((Item=(Type=Bone,Name="deform_knee_fr_01_l"),Transform=(Rotation=(X=0,Y=0,Z=0,W=1),Translation=(X=0.258,Y=-3.591,Z=0.009),Scale3D=(X=1,Y=1,Z=1))))
    Mode=OverrideLocal
    WeightMinimum=0
    Weight=@n15.Result
    ItemToModify.0.Transform.Rotation=&RigVMFunction_MathQuaternionFromEuler(Euler=&RigVMFunction_MathVectorAdd(B=(X=0,Y=0,Z=-51.2), A=@n14.Result) {template="Add::Execute(in A,in B,out Result)"}.Result).Result
   )
   n16 = &RigVMFunction_MathQuaternionToEuler(
    RotationOrder=ZYX
    Value=&RigUnit_GetRelativeTransformForItem(Child=(Type=Bone,Name="calf_r"), Parent=(Type=Bone,Name="thigh_r"), bChildInitial=False, bParentInitial=False).RelativeTransform.Rotation
   )
   n17 = &RigVMFunction_MathFloatRemap(SourceMaximum=90, SourceMinimum=4, TargetMinimum=0, bClamp=true, Value=@n16.Result.Z)
   n18 = &RigVMFunction_MathQuaternionToEuler(
    RotationOrder=ZYX
    Value=&RigUnit_GetTransform(Item=(Type=Bone,Name="deform_knee_bk_01_r"), Space=LocalSpace, bInitial=true).Transform.Rotation
   )
   > &RigUnit_ModifyTransforms(
    ItemToModify=((Item=(Type=Bone,Name="deform_knee_bk_01_r"),Transform=(Rotation=(X=0,Y=0,Z=0,W=1),Translation=(X=4.319,Y=4.798,Z=0),Scale3D=(X=1,Y=2.022,Z=1.95))))
    Mode=OverrideLocal
    WeightMinimum=0
    Weight=@n17.Result
    ItemToModify.0.Transform.Rotation=&RigVMFunction_MathQuaternionFromEuler(Euler=&RigVMFunction_MathVectorAdd(B=(X=0,Y=0,Z=0), A=@n18.Result) {template="Add::Execute(in A,in B,out Result)"}.Result).Result
   )
   n19 = &RigVMFunction_MathQuaternionToEuler(
    RotationOrder=ZYX
    Value=&RigUnit_GetTransform(Item=(Type=Bone,Name="deform_knee_fr_01_r"), Space=LocalSpace, bInitial=true).Transform.Rotation
   )
   > &RigUnit_ModifyTransforms(
    ItemToModify=((Item=(Type=Bone,Name="deform_knee_fr_01_r"),Transform=(Rotation=(X=0,Y=0,Z=0,W=1),Translation=(X=-1.701,Y=-1.197,Z=0),Scale3D=(X=1,Y=1,Z=1))))
    Mode=OverrideLocal
    WeightMinimum=0
    Weight=@n17.Result
    ItemToModify.0.Transform.Rotation=&RigVMFunction_MathQuaternionFromEuler(Euler=&RigVMFunction_MathVectorAdd(B=(X=0,Y=0,Z=0), A=@n19.Result) {template="Add::Execute(in A,in B,out Result)"}.Result).Result
   )
   n20 = &RigVMFunction_MathFloatRemap(SourceMaximum=130, SourceMinimum=89, TargetMinimum=0, bClamp=true, Value=@n16.Result.Z)
   > &RigUnit_ModifyTransforms(
    ItemToModify=((Item=(Type=Bone,Name="deform_knee_bk_01_r"),Transform=(Rotation=(X=0,Y=0,Z=0,W=1),Translation=(X=10.633,Y=3.561,Z=0),Scale3D=(X=1,Y=1,Z=1.777))))
    Mode=OverrideLocal
    WeightMinimum=0
    Weight=@n20.Result
    ItemToModify.0.Transform.Rotation=&RigVMFunction_MathQuaternionFromEuler(Euler=&RigVMFunction_MathVectorAdd(B=(X=0,Y=0,Z=0), A=@n18.Result) {template="Add::Execute(in A,in B,out Result)"}.Result).Result
   )
   > &RigUnit_ModifyTransforms(
    ItemToModify=((Item=(Type=Bone,Name="deform_knee_fr_01_r"),Transform=(Rotation=(X=0,Y=0,Z=0,W=1),Translation=(X=-1.791,Y=-1.35,Z=0),Scale3D=(X=1.913,Y=1,Z=1))))
    Mode=OverrideLocal
    WeightMinimum=0
    Weight=@n20.Result
    ItemToModify.0.Transform.Rotation=&RigVMFunction_MathQuaternionFromEuler(Euler=&RigVMFunction_MathVectorAdd(B=(X=0,Y=0,Z=0), A=@n19.Result) {template="Add::Execute(in A,in B,out Result)"}.Result).Result
   )
   n21 = &RigUnit_GetRelativeTransformForItem(Child=(Type=Bone,Name="upperarm_l"), Parent=(Type=Bone,Name="spine_05"), bChildInitial=False, bParentInitial=False)
   n22 = &RigVMFunction_MathFloatRemap(
    SourceMaximum=50
    SourceMinimum=170
    TargetMinimum=0
    bClamp=true
    Value=/Script/RigVM.RigVMFunction_MathFloatAbs(Value=&RigVMFunction_MathQuaternionToEuler(RotationOrder=ZYX, Value=@n21.RelativeTransform.Rotation).Result.Z) {is=struct RigVMUnitNode, template="Absolute::Execute(in Value,out Result)"}.Result
   )
   n23 = &RigVMFunction_MathQuaternionToEuler(
    RotationOrder=ZYX
    Value=&RigUnit_GetTransform(Item=(Type=Bone,Name="deform_pectoral_fr_l"), Space=LocalSpace, bInitial=true).Transform.Rotation
   )
   > &RigUnit_ModifyTransforms(
    ItemToModify=((Item=(Type=Bone,Name="deform_pectoral_fr_l"),Transform=(Rotation=(X=0,Y=0,Z=0,W=1),Translation=(X=6.262,Y=5.9,Z=4.98),Scale3D=(X=1,Y=1,Z=1))))
    Mode=OverrideLocal
    WeightMinimum=0
    Weight=@n22.Result
    ItemToModify.0.Transform.Rotation=&RigVMFunction_MathQuaternionFromEuler(Euler=&RigVMFunction_MathVectorAdd(B=(X=0,Y=0,Z=0), A=@n23.Result) {template="Add::Execute(in A,in B,out Result)"}.Result).Result
   )
   n24 = &RigVMFunction_MathQuaternionToEuler(
    RotationOrder=ZYX
    Value=&RigUnit_GetTransform(Item=(Type=Bone,Name="deform_deltoid_fr_l"), Space=LocalSpace, bInitial=true).Transform.Rotation
   )
   > &RigUnit_ModifyTransforms(
    ItemToModify=((Item=(Type=Bone,Name="deform_deltoid_fr_l"),Transform=(Rotation=(X=0,Y=0,Z=0,W=1),Translation=(X=5.008,Y=3.542,Z=0),Scale3D=(X=1,Y=1,Z=1))))
    Mode=OverrideLocal
    WeightMinimum=0
    Weight=@n22.Result
    ItemToModify.0.Transform.Rotation=&RigVMFunction_MathQuaternionFromEuler(Euler=&RigVMFunction_MathVectorAdd(B=(X=0,Y=0,Z=-51), A=@n24.Result) {template="Add::Execute(in A,in B,out Result)"}.Result).Result
   )
   n25 = &RigUnit_GetRelativeTransformForItem(Child=(Type=Bone,Name="upperarm_r"), Parent=(Type=Bone,Name="spine_05"), bChildInitial=False, bParentInitial=False)
   n26 = &RigVMFunction_MathFloatRemap(
    SourceMaximum=123
    SourceMinimum=30
    TargetMinimum=0
    bClamp=true
    Value=&RigVMFunction_MathFloatAbs(Value=&RigVMFunction_MathQuaternionToEuler(RotationOrder=ZYX, Value=@n25.RelativeTransform.Rotation).Result.Z) {template="Absolute::Execute(in Value,out Result)"}.Result
   )
   RigUnit_GetTransform = &RigUnit_GetTransform(Item=(Type=Bone,Name="deform_pectoral_fr_r"), Space=LocalSpace, bInitial=true)
   n27 = &RigVMFunction_MathQuaternionToEuler(RotationOrder=ZYX, Value=@RigUnit_GetTransform.Transform.Rotation)
   > &RigUnit_ModifyTransforms(
    ItemToModify=((Item=(Type=Bone,Name="deform_pectoral_fr_r"),Transform=(Rotation=(X=0,Y=0,Z=0,W=1),Translation=(X=-6.262,Y=-5.9,Z=-4.98),Scale3D=(X=1,Y=1,Z=1))))
    Mode=OverrideLocal
    WeightMinimum=0
    Weight=@n26.Result
    ItemToModify.0.Transform.Rotation=&RigVMFunction_MathQuaternionFromEuler(Euler=&RigVMFunction_MathVectorAdd(B=(X=0,Y=0,Z=0), A=@n27.Result) {template="Add::Execute(in A,in B,out Result)"}.Result).Result
   )
   n28 = &RigVMFunction_MathQuaternionToEuler(
    RotationOrder=ZYX
    Value=&RigUnit_GetTransform(Item=(Type=Bone,Name="deform_deltoid_fr_r"), Space=LocalSpace, bInitial=true).Transform.Rotation
   )
   > &RigUnit_ModifyTransforms(
    ItemToModify=((Item=(Type=Bone,Name="deform_deltoid_fr_r"),Transform=(Rotation=(X=0,Y=0,Z=0,W=1),Translation=(X=-5.008,Y=-3.542,Z=0),Scale3D=(X=1,Y=1,Z=1))))
    Mode=OverrideLocal
    WeightMinimum=0
    Weight=@n26.Result
    ItemToModify.0.Transform.Rotation=&RigVMFunction_MathQuaternionFromEuler(Euler=&RigVMFunction_MathVectorAdd(B=(X=0,Y=0,Z=-51), A=@n28.Result) {template="Add::Execute(in A,in B,out Result)"}.Result).Result
   )
   n29 = &RigUnit_GetRelativeTransformForItem(Child=(Type=Bone,Name="thigh_l"), Parent=(Type=Bone,Name="pelvis"), bChildInitial=False, bParentInitial=False)
   n30 = &RigVMFunction_MathFloatRemap(
    SourceMaximum=100
    SourceMinimum=2
    TargetMinimum=0
    bClamp=true
    Value=&RigVMFunction_MathFloatAbs(Value=&RigVMFunction_MathQuaternionToEuler(RotationOrder=ZYX, Value=@n29.RelativeTransform.Rotation).Result.Z) {template="Absolute::Execute(in Value,out Result)"}.Result
   )
   n31 = &RigVMFunction_MathQuaternionToEuler(
    RotationOrder=ZYX
    Value=&RigUnit_GetTransform(Item=(Type=Bone,Name="deform_thigh_fr_l"), Space=LocalSpace, bInitial=true).Transform.Rotation
   )
   > &RigUnit_ModifyTransforms(
    ItemToModify=((Item=(Type=Bone,Name="deform_thigh_fr_l"),Transform=(Rotation=(X=0,Y=0,Z=0,W=1),Translation=(X=13.423,Y=8.962,Z=2.253),Scale3D=(X=1,Y=1,Z=1.175))))
    Mode=OverrideLocal
    WeightMinimum=0
    Weight=@n30.Result
    ItemToModify.0.Transform.Rotation=&RigVMFunction_MathQuaternionFromEuler(Euler=&RigVMFunction_MathVectorAdd(B=(X=0,Y=-14.949,Z=0), A=@n31.Result) {template="Add::Execute(in A,in B,out Result)"}.Result).Result
   )
   n32 = &RigVMFunction_MathQuaternionToEuler(
    RotationOrder=ZYX
    Value=&RigUnit_GetTransform(Item=(Type=Bone,Name="deform_thigh_bk_l"), Space=LocalSpace, bInitial=true).Transform.Rotation
   )
   > &RigUnit_ModifyTransforms(
    ItemToModify=((Item=(Type=Bone,Name="deform_thigh_bk_l"),Transform=(Rotation=(X=0,Y=0,Z=0,W=1),Translation=(X=29.958,Y=-1.401,Z=0.029),Scale3D=(X=1,Y=1,Z=1.803))))
    Mode=OverrideLocal
    WeightMinimum=0
    Weight=@n30.Result
    ItemToModify.0.Transform.Rotation=&RigVMFunction_MathQuaternionFromEuler(Euler=&RigVMFunction_MathVectorAdd(B=(X=0,Y=-37.217,Z=0), A=@n32.Result) {template="Add::Execute(in A,in B,out Result)"}.Result).Result
   )
   n33 = &RigUnit_GetRelativeTransformForItem(Child=(Type=Bone,Name="thigh_r"), Parent=(Type=Bone,Name="pelvis"), bChildInitial=False, bParentInitial=False)
   n34 = &RigVMFunction_MathFloatRemap(
    SourceMaximum=77
    SourceMinimum=177
    TargetMinimum=0
    bClamp=true
    Value=&RigVMFunction_MathFloatAbs(Value=&RigVMFunction_MathQuaternionToEuler(RotationOrder=ZYX, Value=@n33.RelativeTransform.Rotation).Result.Z) {template="Absolute::Execute(in Value,out Result)"}.Result
   )
   n35 = &RigVMFunction_MathQuaternionToEuler(
    RotationOrder=ZYX
    Value=&RigUnit_GetTransform(Item=(Type=Bone,Name="deform_thigh_fr_r"), Space=LocalSpace, bInitial=true).Transform.Rotation
   )
   > &RigUnit_ModifyTransforms(
    ItemToModify=((Item=(Type=Bone,Name="deform_thigh_fr_r"),Transform=(Rotation=(X=0,Y=0,Z=0,W=1),Translation=(X=-13.423,Y=-8.962,Z=-2.253),Scale3D=(X=1,Y=1,Z=1.175))))
    Mode=OverrideLocal
    WeightMinimum=0
    Weight=@n34.Result
    ItemToModify.0.Transform.Rotation=&RigVMFunction_MathQuaternionFromEuler(Euler=&RigVMFunction_MathVectorAdd(B=(X=0,Y=-14.949,Z=0), A=@n35.Result) {template="Add::Execute(in A,in B,out Result)"}.Result).Result
   )
   n36 = &RigVMFunction_MathQuaternionToEuler(
    RotationOrder=ZYX
    Value=&RigUnit_GetTransform(Item=(Type=Bone,Name="deform_thigh_bk_r"), Space=LocalSpace, bInitial=true).Transform.Rotation
   )
   > &RigUnit_ModifyTransforms(
    ItemToModify=((Item=(Type=Bone,Name="deform_thigh_bk_r"),Transform=(Rotation=(X=0,Y=0,Z=0,W=1),Translation=(X=-29.958,Y=1.401,Z=-0.029),Scale3D=(X=1,Y=1,Z=1.803))))
    Mode=OverrideLocal
    WeightMinimum=0
    Weight=@n34.Result
    ItemToModify.0.Transform.Rotation=&RigVMFunction_MathQuaternionFromEuler(Euler=&RigVMFunction_MathVectorAdd(B=(X=0,Y=-37.217,Z=0), A=@n36.Result) {template="Add::Execute(in A,in B,out Result)"}.Result).Result
   )
   > RigVMFunctionReturnNode()
   /Script/ControlRig.RigUnit_DebugHierarchy(
    Color=(R=1,G=1,B=1,A=1)
    DebugDrawSettings=(DepthPriority=SDPG_Foreground,Lifetime=-1)
    ExecutePin=()
    Items=()
    Scale=3
    Thickness=0.1
    WorldOffset=(Rotation=(X=0,Y=0,Z=0,W=1),Translation=(X=0,Y=0,Z=0),Scale3D=(X=1,Y=1,Z=1))
    bEnabled=True
   ) {is=struct RigVMUnitNode}
  ]}
 CollapseNode_ContainedGraph {nodes [
   n1 = RigVMFunctionEntryNode()
   /Script/ControlRigDynamics.RigUnit_SpawnDynamicsConstraint(
    ChildComponentKey=(ElementKey=(Type=Bone,Name="dyn_tubes_top_r_02"),Name="DynamicsParticle_tubes")
    ConstraintComponentName=DynamicsConstraint_tubes
    ConstraintType=Soft
    DampingRatio=1
    DynamicsSolverComponentKey=(ElementKey=(Type=Bone,Name="root"),Name="DynamicsSolver_tubes")
    ExtraDamping=0
    ExtraLength=0
    LengthMultiplier=1
    Owner=(Type=Bone,Name="upperarm_r")
    ParentComponentKey=(ElementKey=(Type=Bone,Name="dyn_tubes_top_r_01"),Name="DynamicsParticle_tubes")
    Strength=0
    bAccelerationMode=True
    ExecutePin=@n1.ExecutePin
   ) {is=struct RigVMUnitNode}
   > n2 = &RigUnit_SpawnDynamicsConstraint(
    ChildComponentKey=(ElementKey=(Type=Bone,Name="dyn_tubes_bott_r_02"),Name="DynamicsParticle_tubes")
    ConstraintComponentName=DynamicsConstraint_tubes
    ConstraintType=Soft
    DampingRatio=1
    DynamicsSolverComponentKey=(ElementKey=(Type=Bone,Name="root"),Name="DynamicsSolver_tubes")
    ExtraDamping=0
    ExtraLength=0
    LengthMultiplier=1
    Owner=(Type=Bone,Name="upperarm_r")
    ParentComponentKey=(ElementKey=(Type=Bone,Name="dyn_tubes_bott_r_01"),Name="DynamicsParticle_tubes")
    Strength=0
    bAccelerationMode=True
   )
   RigVMFunctionReturnNode(ExecutePin=@n2.ExecutePin)
  ]}
 CollapseNode_ContainedGraph {nodes [
   n1 = RigVMFunctionEntryNode()
   n2 = /Script/ControlRig.RigUnit_HierarchyAddBone(
    Name=dyn_holster_top_end
    Parent=(Type=Bone,Name="dyn_holster_top")
    Space=LocalSpace
    Transform=(Rotation=(X=0,Y=0,Z=0,W=1),Translation=(X=30,Y=0,Z=0),Scale3D=(X=1,Y=1,Z=1))
    ExecutePin=@n1.ExecutePin
   ) {is=struct RigVMUnitNode}
   RigVMFunctionReturnNode(ExecutePin=@n2.ExecutePin)
  ]}
 New Function_ContainedGraph {nodes [
   RigVMFunctionEntryNode()
   > /Script/ControlRigDynamics.RigUnit_SpawnDynamicsConfiner(
    ConfinedParticles=((ElementKey=(Type=Bone,Name="dyn_cape_mid_02"),Name="DynamicsParticle"))
    ConfinerComponentName=DynamicsConfiner
    DynamicsSolverComponentKey=(ElementKey=(Type=Bone,Name="Root"),Name="DynamicsSolver")
    Owner=(Type=Bone,Name="spine_03")
    Shapes=(Boxes=((Name="box",TM=(Rotation=(X=0,Y=0,Z=0,W=1),Translation=(X=0,Y=-30,Z=0),Scale3D=(X=1,Y=1,Z=1)),Extents=(X=50,Y=50,Z=0))))
    Strength=10
   ) {is=struct RigVMUnitNode}
   > &RigUnit_SpawnDynamicsConfiner(
    ConfinedParticles=((ElementKey=(Type=Bone,Name="dyn_cape_mid_03"),Name="DynamicsParticle"))
    ConfinerComponentName=DynamicsConfiner
    DynamicsSolverComponentKey=(ElementKey=(Type=Bone,Name="Root"),Name="DynamicsSolver")
    Owner=(Type=Bone,Name="spine_02")
    Shapes=(Boxes=((Name="box",TM=(Rotation=(X=0,Y=0,Z=0,W=1),Translation=(X=0,Y=-30,Z=0),Scale3D=(X=1,Y=1,Z=1)),Extents=(X=50,Y=50,Z=0))))
    Strength=10
   )
   > &RigUnit_SpawnDynamicsConfiner(
    ConfinedParticles=((ElementKey=(Type=Bone,Name="dyn_cape_mid_04"),Name="DynamicsParticle"),(ElementKey=(Type=Bone,Name="dyn_cape_mid_05"),Name="DynamicsParticle"))
    ConfinerComponentName=DynamicsConfiner
    DynamicsSolverComponentKey=(ElementKey=(Type=Bone,Name="Root"),Name="DynamicsSolver")
    Owner=(Type=Bone,Name="spine_01")
    Shapes=(Boxes=((Name="box",TM=(Rotation=(X=0,Y=0,Z=0,W=1),Translation=(X=0,Y=-30,Z=0),Scale3D=(X=1,Y=1,Z=1)),Extents=(X=50,Y=50,Z=0))))
    Strength=10
   )
   > RigVMFunctionReturnNode()
  ]}
 RigVMFunctionLibrary {nodes [MakeConfiners = RigVMCollapseNode(ExecuteContext=())]}
 RigVMFunction_Sequence_1_ContainedGraph {nodes [
   n1 = RigVMFunctionEntryNode()
   RigVMFunction_Sequence = &RigVMFunction_Sequence(ExecuteContext=@n1.ExecuteContext)
   n2 = &RigVMFunction_Sequence(ExecuteContext=@RigVMFunction_Sequence.B)
   RigVMFunctionReturnNode(A=@RigVMFunction_Sequence.A, B=@n2.A, C=@n2.B)
  ]}
 RigVMModel {nodes [
   n1 = &RigUnit_PrepareForExecution()
   n2 = &RigVMFunction_Sequence(ExecuteContext=@n1.ExecutePin)
   /Script/ControlRigDynamics.RigUnit_SpawnDynamicsSolver1(
    Owner=(Type=Bone,Name="Root")
    Settings="(SimulationSpace=SpaceBone,ReferenceBone=(Type=Bone,Name=\"upperarm_r\"),Gravity=(X=0,Y=0,Z=-980),MaxTimeStep=0.0333,MaxNumSteps=2,NumIterations=2,NumConstraintSubIterations=2,bReadBoneOrientations=True,bReadBonePositions=False,bResetFromPosition=False,Posit ...(truncated)"
    SolverComponentName=DynamicsSolver_tubes
    SpaceMotion="(ComponentTransformSource=StartOfFrameSnapshot,ScaleAndClampingSpace=SpaceBone,LinearMotionScale=(X=1,Y=1,Z=1),bClampLinearVelocity=True,MaxLinearVelocityClamp=(X=800,Y=800,Z=800),bClampLinearAcceleration=True,MaxLinearAccelerationClamp=(X=1600,Y=1600,Z=16 ...(truncated)"
    TeleportDetection=(bFromPositionChange=True,PositionChangeThreshold=100,bFromLinearAcceleration=True,LinearAccelerationThreshold=100000,bFromOrientationChange=True,OrientationChangeThreshold=30,bFromAngularAcceleration=True,AngularAccelerationThreshold=100000)
    ExecutePin=@n2.A
   ) {is=struct RigVMUnitNode}
   > n3 = /Script/ControlRigDynamics.RigUnit_SpawnDynamicsChains(
    ChainCurves="(RadiusMultiplier=(EditorCurveData=(Keys=((Value=1),(Time=1,Value=1)),DefaultValue=340282346638528859811704183484516925440,PreInfinityExtrap=RCCE_Constant,PostInfinityExtrap=RCCE_Constant),ExternalCurve=None),MassMultiplier=(EditorCurveData=(Keys=((Value=1 ...(truncated)"
    ChainTerminatorBones=()
    ColliderComponentName=DynamicsCollider_tubes
    Colliders=()
    DynamicsSolverComponentKey=(ElementKey=(Type=Bone,Name="Root"),Name="DynamicsSolver_tubes")
    ElementExcludeMask=()
    ElementIncludeMask=()
    ExtraBones=()
    ParticleComponentName=DynamicsParticle_tubes
    ParticleProperties="(Radius=1,Mass=1,MovementType=Simulated,GravityMultiplier=0,Strength=2,DampingRatio=1,ExtraDamping=0,bAccelerationMode=True,TargetVelocityInfluence=0,TargetMode=1,AngleLimit=45,AngleLimitStrength=10,Damping=0,bScaleDampingByInverseMass=False,bCollideWithCo ...(truncated)"
    RootBones=((Type=Bone,Name="dyn_tubes_top_r_02_MSTR"),(Type=Bone,Name="dyn_tubes_bott_r_02_MSTR"),(Type=Bone,Name="dyn_tubes_top_r_MSTR"),(Type=Bone,Name="dyn_tubes_bott_r_MSTR"))
   ) {is=struct RigVMUnitNode}
   MakeCrossConstraints = RigVMCollapseNode(ExecutePin=@n3.ExecutePin)
   MakeEndBones_holster = RigVMCollapseNode(ExecutePin=@n2.B)
   &RigUnit_SpawnDynamicsSolver1(
    Owner=(Type=Bone,Name="Root")
    Settings="(SimulationSpace=SpaceBone,ReferenceBone=(Type=Bone,Name=\"pelvis\"),Gravity=(X=0,Y=0,Z=-980),MaxTimeStep=0.0333,MaxNumSteps=2,NumIterations=2,NumConstraintSubIterations=2,bReadBoneOrientations=True,bReadBonePositions=False,bResetFromPosition=False,PositionT ...(truncated)"
    SolverComponentName=DynamicsSolver_holster
    SpaceMotion="(ComponentTransformSource=StartOfFrameSnapshot,ScaleAndClampingSpace=SpaceBone,LinearMotionScale=(X=1,Y=1,Z=1),bClampLinearVelocity=True,MaxLinearVelocityClamp=(X=800,Y=800,Z=800),bClampLinearAcceleration=True,MaxLinearAccelerationClamp=(X=1600,Y=1600,Z=16 ...(truncated)"
    TeleportDetection=(bFromPositionChange=True,PositionChangeThreshold=100,bFromLinearAcceleration=True,LinearAccelerationThreshold=100000,bFromOrientationChange=True,OrientationChangeThreshold=30,bFromAngularAcceleration=True,AngularAccelerationThreshold=100000)
    ExecutePin=@MakeEndBones_holster.ExecutePin
   )
   > &RigUnit_SpawnDynamicsChains(
    ChainCurves="(RadiusMultiplier=(EditorCurveData=(Keys=((Value=1),(Time=1,Value=1)),DefaultValue=340282346638528859811704183484516925440,PreInfinityExtrap=RCCE_Constant,PostInfinityExtrap=RCCE_Constant),ExternalCurve=None),MassMultiplier=(EditorCurveData=(Keys=((Value=1 ...(truncated)"
    ChainTerminatorBones=()
    ColliderComponentName=DynamicsCollider_holster
    Colliders=()
    DynamicsSolverComponentKey=(ElementKey=(Type=Bone,Name="Root"),Name="DynamicsSolver_holster")
    ElementExcludeMask=()
    ElementIncludeMask=()
    ExtraBones=()
    ParticleComponentName=DynamicsParticle_holster
    ParticleProperties="(Radius=5,Mass=1,MovementType=Simulated,GravityMultiplier=1,Strength=0.1,DampingRatio=0.1,ExtraDamping=0,bAccelerationMode=True,TargetVelocityInfluence=0,TargetMode=1,AngleLimit=15,AngleLimitStrength=10,Damping=0,bScaleDampingByInverseMass=False,bCollideWi ...(truncated)"
    RootBones=((Type=Bone,Name="dyn_holster_top"))
   )
   > &RigUnit_SpawnDynamicsConfiner(
    ConfinedParticles=((ElementKey=(Type=Bone,Name="dyn_holster_top_end"),Name="DynamicsParticle_holster"))
    ConfinerComponentName=DynamicsConfiner_holster
    DynamicsSolverComponentKey=(ElementKey=(Type=Bone,Name="Root"),Name="DynamicsSolver_holster")
    Owner=(Type=Bone,Name="dyn_holster_bott")
    Shapes=(Boxes=((Name="box",TM=(Rotation=(X=0,Y=0,Z=0,W=1),Translation=(X=0,Y=0,Z=0),Scale3D=(X=1,Y=1,Z=1)),Extents=(X=30,Y=0.1,Z=0.1))))
    Strength=3
   )
   RigUnit_BeginExecution = &RigUnit_BeginExecution()
   RigVMFunction_Sequence = RigVMAggregateNode(ExecuteContext=@RigUnit_BeginExecution.ExecutePin)
   Old Deform Rig Sculpters = RigVMCollapseNode(ExecuteContext=@RigVMFunction_Sequence.A)
   &RigUnit_ParentConstraint(
    AdvancedSettings=(InterpolationType=Average,RotationOrderForFilter=XZY)
    Child=(Type=Bone,Name="dyn_holster_bott_MSTR")
    Parents="((Item=(Type=Bone,Name=\"thigh_r\"),Weight=0.5357),(Item=(Type=Bone,Name=\"thigh_twist_01_r\"),Weight=0.3307),(Item=(Type=Bone,Name=\"deform_knee_fr_01_r\"),Weight=0.0468),(Item=(Type=Bone,Name=\"deform_knee_bk_01_r\"),Weight=0.041),(Item=(Type=Bone,Name=\"pelvis\") ...(truncated)"
    Weight=1
    ExecutePin=@RigVMFunction_Sequence.B
   )
   /Script/ControlRigDynamics.RigUnit_StepDynamicsSolver(
    Alpha=1
    DeltaTimeOverride=0
    DynamicsSolverComponentKey=(ElementKey=(Type=Bone,Name="Root"),Name="DynamicsSolver_tubes")
    SimulationSpaceDeltaTimeOverride=0
    VisualizationSettings=(LineThickness=0.1,ShapeDetail=16,bShowParticles=False,bShowColliders=False,bShowConfiners=False,bShowSkeletalConstraints=False,bShowHardConstraints=False,bShowSoftConstraints=False,bShowAngleLimits=False,bShowConeLimits=False,ParticleValueDisplay=None)
    bTrackVelocitiesDuringPassThrough=True
    ExecutePin=@RigVMFunction_Sequence.C
   ) {is=struct RigVMUnitNode}
   > &RigUnit_StepDynamicsSolver(
    Alpha=1
    DeltaTimeOverride=0
    DynamicsSolverComponentKey=(ElementKey=(Type=Bone,Name="Root"),Name="DynamicsSolver_holster")
    SimulationSpaceDeltaTimeOverride=0
    VisualizationSettings=(LineThickness=0.1,ShapeDetail=16,bShowParticles=False,bShowColliders=False,bShowConfiners=False,bShowSkeletalConstraints=False,bShowHardConstraints=False,bShowSoftConstraints=False,bShowAngleLimits=False,bShowConeLimits=False,ParticleValueDisplay=None)
    bTrackVelocitiesDuringPassThrough=True
   )
  ]}
]
hierarchy [
 Root {loc=<0 0 0>, rot=<0 0 0 1>}
 pelvis {parent=Root, loc=<0 -0.7071 93.1191>, rot=<0.0127 0.707 -0.0127 -0.707>}
 Spine_01 {parent=pelvis, loc=<6.7494 -0.6388 0>, rot=<0 0 0.034 -0.9994>}
 Spine_02 {parent=Spine_01, loc=<7.2518 -0.035 0>, rot=<0 0 0.0238 -0.9997>}
 Spine_03 {parent=Spine_02, loc=<6.0104 0 0>, rot=<0 0 0.0149 -0.9999>}
 spine_04 {parent=Spine_03, loc=<8.1024 0 0>, rot=<0 0 -0.0132 -0.9999>}
 spine_05 {parent=spine_04, loc=<10.0146 0 0>, rot=<0 0 -0.0015 -1>}
 clavicle_l {parent=spine_05, loc=<4.3575 -0.1466 -2.3223>, rot=<0.0465 -0.7289 -0.009 -0.6829>}
 upperarm_l {parent=clavicle_l, loc=<15.5111 0 0>, rot=<0.0226 -0.3234 -0.0088 -0.9459>}
 lowerarm_l {parent=upperarm_l, loc=<26.6514 0 0>, rot=<0.0391 -0.0099 -0.2453 -0.9686>}
 hand_l {parent=lowerarm_l, loc=<22.3491 0 0>, rot=<-0.7081 -0.0507 0.0975 -0.6975>}
 index_metacarpal_l {parent=hand_l, loc=<1.6535 0.0183 -2.96>, rot=<0.0158 -0.0124 -0.0173 -0.9996>}
 index_01_l {parent=index_metacarpal_l, loc=<8.3008 0.0357 -0.0524>, rot=<-0.0044 0.0485 0.0924 -0.9945>}
 index_02_l {parent=index_01_l, loc=<3.3587 -0.3004 0.0452>, rot=<-0.002 0.0027 0.3225 -0.9466>}
 index_03_l {parent=index_02_l, loc=<2.5578 -0.123 -0.0252>, rot=<-0.002 0.0067 0.2558 -0.9667>}
 middle_metacarpal_l {parent=hand_l, loc=<2.1856 -0.1725 -0.5984>, rot=<0.006 0.0168 -0.0329 -0.9993>}
 middle_01_l {parent=middle_metacarpal_l, loc=<7.3275 -0.2447 0.0129>, rot=<-0.0208 0.0042 0.0858 -0.9961>}
 middle_02_l {parent=middle_01_l, loc=<4.0123 0.1401 0.1175>, rot=<-0.0218 0.0258 0.3512 -0.9357>}
 middle_03_l {parent=middle_02_l, loc=<2.6077 0.0534 0.1373>, rot=<0.0082 0.0188 0.1708 -0.9851>}
 pinky_metacarpal_l {parent=hand_l, loc=<1.7616 -0.9753 3.3814>, rot=<-0.0397 0.0814 -0.069 -0.9935>}
 pinky_01_l {parent=pinky_metacarpal_l, loc=<7.1709 -0.4831 -0.1094>, rot=<-0.0654 -0.0035 0.2242 -0.9723>}
 pinky_02_l {parent=pinky_01_l, loc=<2.8274 0.0893 -0.1043>, rot=<0.0037 0.0011 0.2854 -0.9584>}
 pinky_03_l {parent=pinky_02_l, loc=<2.2046 0.0307 -0.0078>, rot=<-0.0002 0.0098 0.1445 -0.9895>}
 ring_metacarpal_l {parent=hand_l, loc=<2.2074 -0.3625 1.3638>, rot=<-0.0284 0.0722 -0.0458 -0.9959>}
 ring_01_l {parent=ring_metacarpal_l, loc=<7.0295 -0.6197 -0.1446>, rot=<-0.0489 -0.0252 0.1892 -0.9804>}
 ring_02_l {parent=ring_01_l, loc=<3.6174 0.4647 -0.1027>, rot=<0 0 0.2806 -0.9598>}
 ring_03_l {parent=ring_02_l, loc=<2.4152 0.114 0.0199>, rot=<0 0 0.1189 -0.9929>}
 thumb_01_l {parent=hand_l, loc=<1.3118 -0.3191 -2.459>, rot=<0.6546 -0.4042 0.0664 -0.6354>}
 thumb_02_l {parent=thumb_01_l, loc=<5.9336 -0.0448 0.0401>, rot=<-0.0037 0.0213 0.298 -0.9543>}
 thumb_03_l {parent=thumb_02_l, loc=<2.2223 0.3682 -0.0907>, rot=<0.0118 -0.013 0.0066 -0.9998>}
 lowerarm_twist_01_l {parent=lowerarm_l, loc=<17.8549 0 0>, rot=<0 0 0 -1>}
 dyn_hand_loc_tracker_l {parent=lowerarm_twist_01_l, loc=<12.3429 -0.589 7.3448>, rot=<-0.7081 -0.0507 0.0975 -0.6975>}
 lowerarm_twist_02_l {parent=lowerarm_l, loc=<9.6199 0 0>, rot=<0 0 0 -1>}
 dyn_gauntlet_l {parent=lowerarm_twist_02_l, loc=<3.4016 0 0>, rot=<0 0 0 1>}
 deform_elbow_bk_l {parent=lowerarm_l, loc=<-1.2853 -2.9274 -0.5607>, rot=<-0.0285 0.0335 -0.0024 -0.999>}
 upperarm_twist_01_l {parent=upperarm_l, loc=<0.0041 0 0>, rot=<0 0 0 -1>}
 deform_deltoid_fr_l {parent=upperarm_twist_01_l, loc=<2.4511 0 0>, rot=<0 0 0 -1>}
 upperarm_twist_02_l {parent=upperarm_l, loc=<13.7941 0 0>, rot=<0 0 0 -1>}
 deform_elbow_fr_l {parent=upperarm_l, loc=<26.5048 0.1875 -0.5816>, rot=<0.0021 0.0156 -0.185 -0.9826>}
 deform_pectoral_fr_l {parent=clavicle_l, loc=<9.3905 0 0>, rot=<0 0 0 1>}
 clavicle_r {parent=spine_05, loc=<4.3575 -0.1466 2.3223>, rot=<0.7289 0.0465 -0.6829 0.009>}
 upperarm_r {parent=clavicle_r, loc=<-15.5111 0 0>, rot=<0.0226 -0.3234 -0.0088 -0.9459>}
 lowerarm_r {parent=upperarm_r, loc=<-26.6515 0 0>, rot=<0.0391 -0.0099 -0.2453 -0.9686>}
 hand_r {parent=lowerarm_r, loc=<-22.3492 0 0>, rot=<-0.7081 -0.0507 0.0975 -0.6975>}
 index_metacarpal_r {parent=hand_r, loc=<-1.6534 -0.0181 2.9601>, rot=<0.0158 -0.0124 -0.0173 -0.9996>}
 index_01_r {parent=index_metacarpal_r, loc=<-8.3008 -0.0357 0.0524>, rot=<-0.0044 0.0485 0.0924 -0.9945>}
 index_02_r {parent=index_01_r, loc=<-3.3589 0.3003 -0.0453>, rot=<-0.002 0.0027 0.3225 -0.9466>}
 index_03_r {parent=index_02_r, loc=<-2.5578 0.1231 0.0251>, rot=<-0.002 0.0067 0.2558 -0.9667>}
 middle_metacarpal_r {parent=hand_r, loc=<-2.1856 0.1726 0.5984>, rot=<0.006 0.0168 -0.0329 -0.9993>}
 middle_01_r {parent=middle_metacarpal_r, loc=<-7.3273 0.2447 -0.0129>, rot=<-0.0208 0.0042 0.0858 -0.9961>}
 middle_02_r {parent=middle_01_r, loc=<-4.0142 -0.1411 -0.07>, rot=<-0.0217 0.0259 0.3512 -0.9357>}
 middle_03_r {parent=middle_02_r, loc=<-2.6074 -0.0534 -0.1312>, rot=<0.0083 0.0188 0.1708 -0.9851>}
 pinky_metacarpal_r {parent=hand_r, loc=<-1.7617 0.9753 -3.3814>, rot=<-0.0397 0.0814 -0.069 -0.9935>}
 pinky_01_r {parent=pinky_metacarpal_r, loc=<-7.1707 0.4832 0.1094>, rot=<-0.0654 -0.0035 0.2242 -0.9723>}
 pinky_02_r {parent=pinky_01_r, loc=<-2.8274 -0.0892 0.1043>, rot=<0.0037 0.0011 0.2854 -0.9584>}
 pinky_03_r {parent=pinky_02_r, loc=<-2.2045 -0.031 0.0078>, rot=<-0.0002 0.0098 0.1445 -0.9895>}
 ring_metacarpal_r {parent=hand_r, loc=<-2.2073 0.3627 -1.3638>, rot=<-0.0284 0.0722 -0.0458 -0.9959>}
 ring_01_r {parent=ring_metacarpal_r, loc=<-7.0297 0.6197 0.1446>, rot=<-0.0489 -0.0252 0.1892 -0.9804>}
 ring_02_r {parent=ring_01_r, loc=<-3.6171 -0.4648 0.1027>, rot=<0 0 0.2806 -0.9598>}
 ring_03_r {parent=ring_02_r, loc=<-2.4153 -0.1141 -0.0199>, rot=<0 0 0.1189 -0.9929>}
 thumb_01_r {parent=hand_r, loc=<-1.3119 0.3192 2.459>, rot=<0.6546 -0.4042 0.0664 -0.6354>}
 thumb_02_r {parent=thumb_01_r, loc=<-5.9336 0.0447 -0.0401>, rot=<-0.0037 0.0213 0.298 -0.9543>}
 thumb_03_r {parent=thumb_02_r, loc=<-2.2223 -0.3682 0.0907>, rot=<0.0118 -0.013 0.0066 -0.9998>}
 lowerarm_twist_01_r {parent=lowerarm_r, loc=<-17.855 0 0>, rot=<0 0 0 1>}
 dyn_hand_loc_tracker_r {parent=lowerarm_twist_01_r, loc=<-12.343 0.5891 -7.3444>, rot=<0.0975 0.6975 0.7081 -0.0507>}
 lowerarm_twist_02_r {parent=lowerarm_r, loc=<-9.6199 0 0>, rot=<0 0 0 1>}
 dyn_gauntlet_r {parent=lowerarm_twist_02_r, loc=<-3.4015 0 0>, rot=<0 -1 0 0>}
 dyn_tubes_top_r_02_MSTR {parent=dyn_gauntlet_r, loc=<-17.6474 -0.6396 7.5062>, rot=<-0.0549 0.075 0.9953 -0.0289>}
 dyn_tubes_top_r_02 {parent=dyn_tubes_top_r_02_MSTR, loc=<5.7375 -1.7539 -1.6325>, rot=<0 0 0 1>}
 dyn_tubes_bott_r_02_MSTR {parent=dyn_gauntlet_r, loc=<-13.7991 5.8547 -0.2046>, rot=<0.0278 -0.7231 -0.6901 -0.0085>}
 dyn_tubes_bott_r_02 {parent=dyn_tubes_bott_r_02_MSTR, loc=<6.4285 -0.3651 0>, rot=<0 -0.1503 0 -0.9886>}
 deform_elbow_bk_r {parent=lowerarm_r, loc=<1.2852 2.9274 0.561>, rot=<-0.0285 0.0335 -0.0024 -0.999>}
 upperarm_twist_01_r {parent=upperarm_r, loc=<-0.0041 0 0>, rot=<0 0 0 1>}
 deform_deltoid_fr_r {parent=upperarm_twist_01_r, loc=<-2.4512 0 0>, rot=<0 0 0 1>}
 dyn_tubes_top_r_MSTR {parent=upperarm_twist_01_r, loc=<-7.2131 6.0469 -3.7069>, rot=<0.0735 -0.8835 -0.4624 -0.0117>}
 dyn_tubes_top_r_01 {parent=dyn_tubes_top_r_MSTR, loc=<5.8912 0 0>, rot=<0 0 0 1>}
 dyn_tubes_bott_r_MSTR {parent=upperarm_twist_01_r, loc=<-10.4531 8.2953 1.2053>, rot=<-0.0025 -0.7122 -0.7017 -0.0195>}
 dyn_tubes_bott_r_01 {parent=dyn_tubes_bott_r_MSTR, loc=<6.3532 0 0>, rot=<0 0 0 1>}
 upperarm_twist_02_r {parent=upperarm_r, loc=<-13.7942 0 0>, rot=<0 0 0 1>}
 deform_elbow_fr_r {parent=upperarm_r, loc=<-26.5046 -0.1875 0.5813>, rot=<0.0021 0.0156 -0.185 -0.9826>}
 deform_pectoral_fr_r {parent=clavicle_r, loc=<-9.39 0 0>, rot=<0 0 0 1>}
 neck_01 {parent=spine_05, loc=<11.3138 0.1529 0>, rot=<0 0 -0.2003 -0.9797>}
 neck_02 {parent=neck_01, loc=<4.5627 0 0>, rot=<0 0 0.0504 -0.9987>}
 head {parent=neck_02, loc=<5.3914 0 0>, rot=<0 0 0.115 -0.9934>}
 dyn_shoulder_flap_bk_r_MSTR {parent=spine_05, loc=<1.8153 -9.6439 15.1368>, rot=<0.5646 -0.5533 0.4838 0.3755>}
 dyn_shoulder_flap_bk_r_01 {parent=dyn_shoulder_flap_bk_r_MSTR, loc=<0.3 0 0>, rot=<0 0 0 1>}
 dyn_shoulder_flap_bk_l_MSTR {parent=spine_05, loc=<1.8156 -9.6438 -15.1368>, rot=<-0.3755 -0.4838 0.5533 -0.5646>}
 dyn_shoulder_flap_bk_l_01 {parent=dyn_shoulder_flap_bk_l_MSTR, loc=<0.3001 -0.0007 0>, rot=<0 0 0 1>}
 attach_backpack {parent=spine_05, loc=<-0.123 -13.473 0>, rot=<-0.0093 -0.707 0.0093 -0.707>}
 Backpack_BR {parent=attach_backpack, type=Null, loc=<0 0.0003 0.0002>, rot=<-0.0131 0 0 0.9999>}
 attach_cape {parent=spine_05, loc=<9.017 -11.714 0>, rot=<0.0616 -0.7044 -0.0616 -0.7044>}
 cape {parent=attach_cape, type=Null, loc=<0 -0.0004 0.0004>, rot=<0.0471 0.0019 -0.0401 0.9981>}
 thigh_l {parent=pelvis, loc=<-2.5693 -0.0109 -7.7501>, rot=<0.0012 0.0572 0.0155 -0.9982>}
 calf_l {parent=thigh_l, loc=<-39.0484 0 0>, rot=<0.0006 -0.0161 -0.0391 -0.9991>}
 foot_l {parent=calf_l, loc=<-42.561 0.1077 0.4292>, rot=<0.0418 -0.0455 0.0343 -0.9975>}
 ball_l {parent=foot_l, loc=<-6.1599 13.8264 0.6806>, rot=<0.0001 -0.0001 -0.7063 -0.7079>}
 calf_twist_01_l {parent=calf_l, loc=<-35.813 -0.0967 0.0397>, rot=<0.0237 0.0095 -0.0035 -0.9997>}
 calf_twist_02_l {parent=calf_l, loc=<-18.8064 -0.0967 0.0397>, rot=<0.0237 0.0095 -0.0035 -0.9997>}
 deform_knee_bk_01_l {parent=calf_l, loc=<0.6022 -1.0607 0>, rot=<0 0 -0.0408 -0.9992>}
 deform_knee_fr_01_l {parent=calf_l, loc=<-0.2342 0 0>, rot=<0 0 0 1>}
 thigh_twist_01_l {parent=thigh_l, loc=<-21.6303 0 0>, rot=<0.0013 0 0 -1>}
 deform_thigh_bk_l {parent=thigh_twist_01_l, loc=<16.9514 0.2058 0.1808>, rot=<-0.0013 -0.0217 0 -0.9998>}
 deform_thigh_fr_l {parent=thigh_twist_01_l, loc=<21.5084 0.2281 -0.0598>, rot=<-0.0013 -0.0217 0 -0.9998>}
 thigh_r {parent=pelvis, loc=<-2.5691 -0.0109 7.7501>, rot=<0.0572 -0.0012 0.9982 0.0155>}
 calf_r {parent=thigh_r, loc=<39.0487 0 0>, rot=<0.0006 -0.0161 -0.0391 -0.9991>}
 foot_r {parent=calf_r, loc=<42.5612 -0.1077 -0.4292>, rot=<0.0418 -0.0455 0.0343 -0.9975>}
 ball_r {parent=foot_r, loc=<6.1598 -13.8264 -0.6806>, rot=<0.0001 -0.0001 -0.7063 -0.7079>}
 calf_twist_01_r {parent=calf_r, loc=<35.8132 0.0967 -0.0397>, rot=<0.0237 0.0095 -0.0035 -0.9997>}
 calf_twist_02_r {parent=calf_r, loc=<18.8064 0.0967 -0.0397>, rot=<0.0237 0.0095 -0.0035 -0.9997>}
 deform_knee_fr_01_r {parent=calf_r, loc=<0.2341 0 0>, rot=<0 0 0 1>}
 deform_knee_bk_01_r {parent=calf_r, loc=<-0.6022 1.0607 0>, rot=<0 0 -0.0408 -0.9992>}
 thigh_twist_01_r {parent=thigh_r, loc=<21.6303 0 0>, rot=<0.0013 0 0 -1>}
 deform_thigh_bk_r {parent=thigh_twist_01_r, loc=<-16.9512 -0.2058 -0.1808>, rot=<-0.0013 -0.0217 0 -0.9998>}
 deform_thigh_fr_r {parent=thigh_twist_01_r, loc=<-21.5083 -0.2281 0.0598>, rot=<-0.0013 -0.0217 0 -0.9998>}
 dyn_holster_top_MSTR {parent=thigh_twist_01_r, loc=<-19.6746 -4.4123 9.3818>, rot=<-0.1024 -0.0252 -0.0252 -0.9941>}
 dyn_holster_top {parent=dyn_holster_top_MSTR, loc=<0.3 0 0>, rot=<0 0 0 1>}
 dyn_holster_bott_MSTR {parent=thigh_r, loc=<30.8247 -2.79 8.0857>, rot=<-0.1011 -0.0253 -0.0252 -0.9942>}
 dyn_holster_bott {parent=dyn_holster_bott_MSTR, loc=<0.3 0 0>, rot=<0 0 0 1>}
 dyn_loincloth_bk_l_MSTR {parent=pelvis, loc=<-2.1025 -8.2901 -12.8601>, rot=<0.0328 -0.8403 0.5372 -0.0653>}
 dyn_loincloth_bk_l_01 {parent=dyn_loincloth_bk_l_MSTR, loc=<0.3 0 0>, rot=<0 0 0 1>}
 dyn_loincloth_bk_l_02 {parent=dyn_loincloth_bk_l_01, loc=<16.7301 0 0>, rot=<0 -0.0811 0 -0.9967>}
 dyn_loincloth_bk_l_03 {parent=dyn_loincloth_bk_l_02, loc=<27.8698 0 0>, rot=<0 0 0 1>}
 dyn_loincloth_bk_MSTR {parent=pelvis, loc=<-1.6846 -12.4003 0>, rot=<0.0283 -0.7065 0.7065 -0.0283>}
 dyn_loincloth_bk_01 {parent=dyn_loincloth_bk_MSTR, loc=<0.3 0 0>, rot=<0 0 0 1>}
 dyn_loincloth_bk_02 {parent=dyn_loincloth_bk_01, loc=<16.7301 0 0>, rot=<0 -0.0628 0 -0.998>}
 dyn_loincloth_bk_03 {parent=dyn_loincloth_bk_02, loc=<27.8698 0 0>, rot=<0 0 0 1>}
 dyn_roincloth_bk_r_MSTR {parent=pelvis, loc=<-2.1026 -8.2901 12.8601>, rot=<0.0653 -0.5372 0.8403 -0.0328>}
 dyn_roincloth_bk_r_01 {parent=dyn_roincloth_bk_r_MSTR, loc=<0.2999 0 0>, rot=<0 0 0 1>}
 dyn_roincloth_bk_r_02 {parent=dyn_roincloth_bk_r_01, loc=<16.7301 0 0>, rot=<0 -0.0811 0 -0.9967>}
 dyn_roincloth_bk_r_03 {parent=dyn_roincloth_bk_r_02, loc=<27.8698 0 0>, rot=<0 0 0 1>}
 r_knee_corrective {type=Curve, loc=<0 0 0>, rot=<0 0 0 1>}
 l_knee_corrective {type=Curve, loc=<0 0 0>, rot=<0 0 0 1>}
 Male_Med_Head_01_Pout {type=Curve, loc=<0 0 0>, rot=<0 0 0 1>}
 Male_Med_Head_02_Pout {type=Curve, loc=<0 0 0>, rot=<0 0 0 1>}
 Male_Med_Head_03_Pout {type=Curve, loc=<0 0 0>, rot=<0 0 0 1>}
 Male_Med_Head_05_Effort {type=Curve, loc=<0 0 0>, rot=<0 0 0 1>}
 Stretch {type=Curve, loc=<0 0 0>, rot=<0 0 0 1>}
 Squash {type=Curve, loc=<0 0 0>, rot=<0 0 0 1>}
 polySurface151 {type=Curve, loc=<0 0 0>, rot=<0 0 0 1>}
 l_cheek_puff {type=Curve, loc=<0 0 0>, rot=<0 0 0 1>}
 r_cheek_puff {type=Curve, loc=<0 0 0>, rot=<0 0 0 1>}
 l_smile_fix {type=Curve, loc=<0 0 0>, rot=<0 0 0 1>}
 r_smile_fix {type=Curve, loc=<0 0 0>, rot=<0 0 0 1>}
 jaw_open_fix {type=Curve, loc=<0 0 0>, rot=<0 0 0 1>}
 r_lip_sneer_fix {type=Curve, loc=<0 0 0>, rot=<0 0 0 1>}
 l_lip_sneer_fix {type=Curve, loc=<0 0 0>, rot=<0 0 0 1>}
 L_Blink {type=Curve, loc=<0 0 0>, rot=<0 0 0 1>}
 R_Blink {type=Curve, loc=<0 0 0>, rot=<0 0 0 1>}
 L_Arm_Stretch {type=Curve, loc=<0 0 0>, rot=<0 0 0 1>}
 Hair_Hat {type=Curve, loc=<0 0 0>, rot=<0 0 0 1>}
 Hair_Helmet {type=Curve, loc=<0 0 0>, rot=<0 0 0 1>}
 Male_Med_Head_02_What {type=Curve, loc=<0 0 0>, rot=<0 0 0 1>}
 Male_Med_Head_02_Smile {type=Curve, loc=<0 0 0>, rot=<0 0 0 1>}
 Male_Med_Head_02_Scowl {type=Curve, loc=<0 0 0>, rot=<0 0 0 1>}
 Male_Med_Head_02_Hmmph {type=Curve, loc=<0 0 0>, rot=<0 0 0 1>}
 Male_Med_Head_02_Grrr {type=Curve, loc=<0 0 0>, rot=<0 0 0 1>}
 Male_Med_Head_02_Effort {type=Curve, loc=<0 0 0>, rot=<0 0 0 1>}
 Male_Med_Head_03_What {type=Curve, loc=<0 0 0>, rot=<0 0 0 1>}
 Male_Med_Head_03_Grrr {type=Curve, loc=<0 0 0>, rot=<0 0 0 1>}
 Male_Med_Head_03_Effort {type=Curve, loc=<0 0 0>, rot=<0 0 0 1>}
 Male_Med_Head_03_Hmmph {type=Curve, loc=<0 0 0>, rot=<0 0 0 1>}
 Male_Med_Head_03_Scowl {type=Curve, loc=<0 0 0>, rot=<0 0 0 1>}
 Male_Med_Head_03_Smile {type=Curve, loc=<0 0 0>, rot=<0 0 0 1>}
 Male_Med_Head_05_Teeth_Low_What {type=Curve, loc=<0 0 0>, rot=<0 0 0 1>}
 Male_Med_Head_05_Teeth_Low_Pout {type=Curve, loc=<0 0 0>, rot=<0 0 0 1>}
 Male_Med_Head_05_Teeth_Low_Hmmph {type=Curve, loc=<0 0 0>, rot=<0 0 0 1>}
 Male_Med_Head_05_Teeth_Low_Grrr {type=Curve, loc=<0 0 0>, rot=<0 0 0 1>}
 Male_Med_Head_05_Teeth_Low_Effort {type=Curve, loc=<0 0 0>, rot=<0 0 0 1>}
 Male_Med_Head_05_What {type=Curve, loc=<0 0 0>, rot=<0 0 0 1>}
 Male_Med_Head_05_Grrr {type=Curve, loc=<0 0 0>, rot=<0 0 0 1>}
 Male_Med_Head_05_Hmmph {type=Curve, loc=<0 0 0>, rot=<0 0 0 1>}
 Male_Med_Head_05_Pout {type=Curve, loc=<0 0 0>, rot=<0 0 0 1>}
 Male_Med_Head_05_Tongue_What {type=Curve, loc=<0 0 0>, rot=<0 0 0 1>}
 Male_Med_Head_05_Tongue_Pout {type=Curve, loc=<0 0 0>, rot=<0 0 0 1>}
 Male_Med_Head_05_Tongue_Hmmph {type=Curve, loc=<0 0 0>, rot=<0 0 0 1>}
 Male_Med_Head_05_Tongue_Grrr {type=Curve, loc=<0 0 0>, rot=<0 0 0 1>}
 Male_Med_Head_05_Tongue_Effort {type=Curve, loc=<0 0 0>, rot=<0 0 0 1>}
 AdditiveLayer {type=Curve, loc=<0 0 0>, rot=<0 0 0 1>}
 RightHandIK {type=Curve, loc=<0 0 0>, rot=<0 0 0 1>}
 LeftHandIK {type=Curve, loc=<0 0 0>, rot=<0 0 0 1>}
 LeftHandFK {type=Curve, loc=<0 0 0>, rot=<0 0 0 1>}
 Male_Med_Head_01_Grrr {type=Curve, loc=<0 0 0>, rot=<0 0 0 1>}
 WeaponAiming {type=Curve, loc=<0 0 0>, rot=<0 0 0 1>}
 Male_Med_Head_01_Hmmph {type=Curve, loc=<0 0 0>, rot=<0 0 0 1>}
 DragonSlash {type=Curve, loc=<0 0 0>, rot=<0 0 0 1>}
 Male_Med_Head_01_Effort {type=Curve, loc=<0 0 0>, rot=<0 0 0 1>}
 LeftHankFK {type=Curve, loc=<0 0 0>, rot=<0 0 0 1>}
 blendOrient1 {type=Curve, loc=<0 0 0>, rot=<0 0 0 1>}
 blendParent1 {type=Curve, loc=<0 0 0>, rot=<0 0 0 1>}
 bottom_R_Corner_Down {type=Curve, loc=<0 0 0>, rot=<0 0 0 1>}
 bottom_R_Corner_Up {type=Curve, loc=<0 0 0>, rot=<0 0 0 1>}
 top_L_Corner_Up {type=Curve, loc=<0 0 0>, rot=<0 0 0 1>}
 top_L_Corner_Down {type=Curve, loc=<0 0 0>, rot=<0 0 0 1>}
 top_R_Corner_Down {type=Curve, loc=<0 0 0>, rot=<0 0 0 1>}
 top_R_Corner_Up {type=Curve, loc=<0 0 0>, rot=<0 0 0 1>}
 bottom_L_Corner_Down {type=Curve, loc=<0 0 0>, rot=<0 0 0 1>}
 bottomt_L_Corner_Up {type=Curve, loc=<0 0 0>, rot=<0 0 0 1>}
 TrailCurve1 {type=Curve, loc=<0 0 0>, rot=<0 0 0 1>}
 BothHandFK {type=Curve, loc=<0 0 0>, rot=<0 0 0 1>}
 bottom_R_Corner_Down1 {type=Curve, loc=<0 0 0>, rot=<0 0 0 1>}
 blendPoint1 {type=Curve, loc=<0 0 0>, rot=<0 0 0 1>}
 WeaponAimingFreeze {type=Curve, loc=<0 0 0>, rot=<0 0 0 1>}
 l_calf_top_taper {type=Curve, loc=<0 0 0>, rot=<0 0 0 1>}
 JogBlendSpaceRotation {type=Curve, loc=<0 0 0>, rot=<0 0 0 1>}
 ZeroOutPitchWeight {type=Curve, loc=<0 0 0>, rot=<0 0 0 1>}
 MeleeTwist {type=Curve, loc=<0 0 0>, rot=<0 0 0 1>}
 MeleeAnim {type=Curve, loc=<0 0 0>, rot=<0 0 0 1>}
 l_sneer_fix {type=Curve, loc=<0 0 0>, rot=<0 0 0 1>}
 jaw_fix {type=Curve, loc=<0 0 0>, rot=<0 0 0 1>}
 r_sneer_fix {type=Curve, loc=<0 0 0>, rot=<0 0 0 1>}
 TurnRotationAmount {type=Curve, loc=<0 0 0>, rot=<0 0 0 1>}
 TurnEnd {type=Curve, loc=<0 0 0>, rot=<0 0 0 1>}
 BlendIn {type=Curve, loc=<0 0 0>, rot=<0 0 0 1>}
 BlendOut {type=Curve, loc=<0 0 0>, rot=<0 0 0 1>}
 BodyRotateX {type=Curve, loc=<0 0 0>, rot=<0 0 0 1>}
 BodyTranslateX {type=Curve, loc=<0 0 0>, rot=<0 0 0 1>}
 CrescentKickScale {type=Curve, loc=<0 0 0>, rot=<0 0 0 1>}
 mask_geo {type=Curve, loc=<0 0 0>, rot=<0 0 0 1>}
 RemoveJumpAdditive {type=Curve, loc=<0 0 0>, rot=<0 0 0 1>}
 RightLeg_IK {type=Curve, loc=<0 0 0>, rot=<0 0 0 1>}
 SnapWeapon_LHand {type=Curve, loc=<0 0 0>, rot=<0 0 0 1>}
 PivotAnim {type=Curve, loc=<0 0 0>, rot=<0 0 0 1>}
 DefaultAdditiveStateWeight {type=Curve, loc=<0 0 0>, rot=<0 0 0 1>}
 StartAnim {type=Curve, loc=<0 0 0>, rot=<0 0 0 1>}
 StopAnim {type=Curve, loc=<0 0 0>, rot=<0 0 0 1>}
 DisableSpeedWarping {type=Curve, loc=<0 0 0>, rot=<0 0 0 1>}
 DisableDBNOSlopeWarp {type=Curve, loc=<0 0 0>, rot=<0 0 0 1>}
 MaskRightArm {type=Curve, loc=<0 0 0>, rot=<0 0 0 1>}
 TurnInPlaceAnim {type=Curve, loc=<0 0 0>, rot=<0 0 0 1>}
 Head_MESH {type=Curve, loc=<0 0 0>, rot=<0 0 0 1>}
 Hair_MESH {type=Curve, loc=<0 0 0>, rot=<0 0 0 1>}
 LaunchPadTransitionOff {type=Curve, loc=<0 0 0>, rot=<0 0 0 1>}
 C_glabella_up_pose {type=Curve, loc=<0 0 0>, rot=<0 0 0 1>}
 jaw_open_pose {type=Curve, loc=<0 0 0>, rot=<0 0 0 1>}
 L_blink_pose {type=Curve, loc=<0 0 0>, rot=<0 0 0 1>}
 L_brow_down_pose {type=Curve, loc=<0 0 0>, rot=<0 0 0 1>}
 L_brow_squeeze_in_pose {type=Curve, loc=<0 0 0>, rot=<0 0 0 1>}
 L_brow_up_pose {type=Curve, loc=<0 0 0>, rot=<0 0 0 1>}
 L_frown_pose {type=Curve, loc=<0 0 0>, rot=<0 0 0 1>}
 L_lip_corner_narrow_pose {type=Curve, loc=<0 0 0>, rot=<0 0 0 1>}
 L_squint_inner_pose {type=Curve, loc=<0 0 0>, rot=<0 0 0 1>}
 L_wide_pose {type=Curve, loc=<0 0 0>, rot=<0 0 0 1>}
 look_down_pose {type=Curve, loc=<0 0 0>, rot=<0 0 0 1>}
 look_left_pose {type=Curve, loc=<0 0 0>, rot=<0 0 0 1>}
 look_right_pose {type=Curve, loc=<0 0 0>, rot=<0 0 0 1>}
 look_up_pose {type=Curve, loc=<0 0 0>, rot=<0 0 0 1>}
 phoneme_fv_pose {type=Curve, loc=<0 0 0>, rot=<0 0 0 1>}
 phoneme_oo_pose {type=Curve, loc=<0 0 0>, rot=<0 0 0 1>}
 R_blink_pose {type=Curve, loc=<0 0 0>, rot=<0 0 0 1>}
 R_brow_down_pose {type=Curve, loc=<0 0 0>, rot=<0 0 0 1>}
 R_brow_squeeze_in_pose {type=Curve, loc=<0 0 0>, rot=<0 0 0 1>}
 R_brow_up_pose {type=Curve, loc=<0 0 0>, rot=<0 0 0 1>}
 R_frown_pose {type=Curve, loc=<0 0 0>, rot=<0 0 0 1>}
 R_lip_corner_narrow_pose {type=Curve, loc=<0 0 0>, rot=<0 0 0 1>}
 R_squint_inner_pose {type=Curve, loc=<0 0 0>, rot=<0 0 0 1>}
 R_wide_pose {type=Curve, loc=<0 0 0>, rot=<0 0 0 1>}
 DisableArmsHeadAdditive {type=Curve, loc=<0 0 0>, rot=<0 0 0 1>}
 FrontendRotation {type=Curve, loc=<0 0 0>, rot=<0 0 0 1>}
 BothLegFK {type=Curve, loc=<0 0 0>, rot=<0 0 0 1>}
 AllowTurnInPlace {type=Curve, loc=<0 0 0>, rot=<0 0 0 1>}
 MontageTimeStretchCurve {type=Curve, loc=<0 0 0>, rot=<0 0 0 1>}
 ParaGlideLeanAlpha {type=Curve, loc=<0 0 0>, rot=<0 0 0 1>}
 WeaponRInChestSpace {type=Curve, loc=<0 0 0>, rot=<0 0 0 1>}
 JimBeam_AO {type=Curve, loc=<0 0 0>, rot=<0 0 0 1>}
 R_squeeze_pose {type=Curve, loc=<0 0 0>, rot=<0 0 0 1>}
 L_squeeze_pose {type=Curve, loc=<0 0 0>, rot=<0 0 0 1>}
 R_upper_lip_raiser_pose {type=Curve, loc=<0 0 0>, rot=<0 0 0 1>}
 L_cheek_up_pose {type=Curve, loc=<0 0 0>, rot=<0 0 0 1>}
 R_cheek_up_pose {type=Curve, loc=<0 0 0>, rot=<0 0 0 1>}
 DisableFaceOverride {type=Curve, loc=<0 0 0>, rot=<0 0 0 1>}
 EnableFullBodyOverride {type=Curve, loc=<0 0 0>, rot=<0 0 0 1>}
 C_glabella_down_pose {type=Curve, loc=<0 0 0>, rot=<0 0 0 1>}
 C_nose_wrinkler_pose {type=Curve, loc=<0 0 0>, rot=<0 0 0 1>}
 eye_convergence_pose {type=Curve, loc=<0 0 0>, rot=<0 0 0 1>}
 jaw_back_pose {type=Curve, loc=<0 0 0>, rot=<0 0 0 1>}
 jaw_forward_pose {type=Curve, loc=<0 0 0>, rot=<0 0 0 1>}
 jaw_left_pose {type=Curve, loc=<0 0 0>, rot=<0 0 0 1>}
 jaw_right_pose {type=Curve, loc=<0 0 0>, rot=<0 0 0 1>}
 jaw_up_pose {type=Curve, loc=<0 0 0>, rot=<0 0 0 1>}
 L_lip_corner_wide_pose {type=Curve, loc=<0 0 0>, rot=<0 0 0 1>}
 L_lower_lip_down_pose {type=Curve, loc=<0 0 0>, rot=<0 0 0 1>}
 L_lower_lip_up_pose {type=Curve, loc=<0 0 0>, rot=<0 0 0 1>}
 L_smile_pose {type=Curve, loc=<0 0 0>, rot=<0 0 0 1>}
 L_upper_lip_lower_pose {type=Curve, loc=<0 0 0>, rot=<0 0 0 1>}
 L_upper_lip_raiser_pose {type=Curve, loc=<0 0 0>, rot=<0 0 0 1>}
 L_upperLidDown_pose {type=Curve, loc=<0 0 0>, rot=<0 0 0 1>}
 mouth_down_pose {type=Curve, loc=<0 0 0>, rot=<0 0 0 1>}
 mouth_up_pose {type=Curve, loc=<0 0 0>, rot=<0 0 0 1>}
 phoneme_ch_pose {type=Curve, loc=<0 0 0>, rot=<0 0 0 1>}
 phoneme_mbp_pose {type=Curve, loc=<0 0 0>, rot=<0 0 0 1>}
 R_lip_corner_wide_pose {type=Curve, loc=<0 0 0>, rot=<0 0 0 1>}
 R_lower_lip_down_pose {type=Curve, loc=<0 0 0>, rot=<0 0 0 1>}
 R_lower_lip_up_pose {type=Curve, loc=<0 0 0>, rot=<0 0 0 1>}
 R_smile_pose {type=Curve, loc=<0 0 0>, rot=<0 0 0 1>}
 R_upper_lip_lower_pose {type=Curve, loc=<0 0 0>, rot=<0 0 0 1>}
 R_upperLidDown_pose {type=Curve, loc=<0 0 0>, rot=<0 0 0 1>}
 tongue_back_pose {type=Curve, loc=<0 0 0>, rot=<0 0 0 1>}
 tongue_down_pose {type=Curve, loc=<0 0 0>, rot=<0 0 0 1>}
 tongue_forward_pose {type=Curve, loc=<0 0 0>, rot=<0 0 0 1>}
 tongue_up_pose {type=Curve, loc=<0 0 0>, rot=<0 0 0 1>}
 LiveLink {type=Curve, loc=<0 0 0>, rot=<0 0 0 1>}
 ParaGlideDeploy {type=Curve, loc=<0 0 0>, rot=<0 0 0 1>}
 HideLeftWeapon {type=Curve, loc=<0 0 0>, rot=<0 0 0 1>}
 DisableLeftHandIKGolfCart {type=Curve, loc=<0 0 0>, rot=<0 0 0 1>}
 DisableRightHandIKGolfCart {type=Curve, loc=<0 0 0>, rot=<0 0 0 1>}
 DisableHandIKRetargeting {type=Curve, loc=<0 0 0>, rot=<0 0 0 1>}
 DisableConsumableOffsetPose {type=Curve, loc=<0 0 0>, rot=<0 0 0 1>}
 UpperBodyNoAdditive {type=Curve, loc=<0 0 0>, rot=<0 0 0 1>}
 MaskLeftArm {type=Curve, loc=<0 0 0>, rot=<0 0 0 1>}
 ForceUpperBody {type=Curve, loc=<0 0 0>, rot=<0 0 0 1>}
 RightHandFK {type=Curve, loc=<0 0 0>, rot=<0 0 0 1>}
 ParaGlideLeanTransition {type=Curve, loc=<0 0 0>, rot=<0 0 0 1>}
 InMotionEmote {type=Curve, loc=<0 0 0>, rot=<0 0 0 1>}
 ForceFullBodyInFrontEnd {type=Curve, loc=<0 0 0>, rot=<0 0 0 1>}
 WeaponLinFloorSpace {type=Curve, loc=<0 0 0>, rot=<0 0 0 1>}
 FrontEnd_Idle_Bypass {type=Curve, loc=<0 0 0>, rot=<0 0 0 1>}
 DisableSlopeWarping {type=Curve, loc=<0 0 0>, rot=<0 0 0 1>}
 Disable_Custom {type=Curve, loc=<0 0 0>, rot=<0 0 0 1>}
 EmoteRequiresWeapon {type=Curve, loc=<0 0 0>, rot=<0 0 0 1>}
 DisableIKRootAdditive {type=Curve, loc=<0 0 0>, rot=<0 0 0 1>}
 DisableFullBodySprint {type=Curve, loc=<0 0 0>, rot=<0 0 0 1>}
 DetachFeet_Vehicle {type=Curve, loc=<0 0 0>, rot=<0 0 0 1>}
 IndigoBoost {type=Curve, loc=<0 0 0>, rot=<0 0 0 1>}
 DisableUpperBodyAdditiveMeshSpace {type=Curve, loc=<0 0 0>, rot=<0 0 0 1>}
 MeleeAOWeight {type=Curve, loc=<0 0 0>, rot=<0 0 0 1>}
 HandsInRootSpace {type=Curve, loc=<0 0 0>, rot=<0 0 0 1>}
 CrouchPelvisAdjustment {type=Curve, loc=<0 0 0>, rot=<0 0 0 1>}
 EnableAccelYawCorrection {type=Curve, loc=<0 0 0>, rot=<0 0 0 1>}
 ParaGlideModifyRoot {type=Curve, loc=<0 0 0>, rot=<0 0 0 1>}
 UseScalarParameterForBlink {type=Curve, loc=<0 0 0>, rot=<0 0 0 1>}
 BlinkControl {type=Curve, loc=<0 0 0>, rot=<0 0 0 1>}
 ForceTargeting {type=Curve, loc=<0 0 0>, rot=<0 0 0 1>}
 BlinkControlProp {type=Curve, loc=<0 0 0>, rot=<0 0 0 1>}
 UseScalarParameterForBlinkProp {type=Curve, loc=<0 0 0>, rot=<0 0 0 1>}
 CrackshotCozyHeadCorrection {type=Curve, loc=<0 0 0>, rot=<0 0 0 1>}
 ForwardMotion {type=Curve, loc=<0 0 0>, rot=<0 0 0 1>}
 EmoteForcedRootYaw {type=Curve, loc=<0 0 0>, rot=<0 0 0 1>}
 MeleeTrailWidth {type=Curve, loc=<0 0 0>, rot=<0 0 0 1>}
 space_world {type=Curve, loc=<0 0 0>, rot=<0 0 0 1>}
 PeelyBlenderFace {type=Curve, loc=<0 0 0>, rot=<0 0 0 1>}
 Width1 {type=Curve, loc=<0 0 0>, rot=<0 0 0 1>}
 OverrideIKHands {type=Curve, loc=<0 0 0>, rot=<0 0 0 1>}
 ParaGlideAddAlpha {type=Curve, loc=<0 0 0>, rot=<0 0 0 1>}
 ParaGlideTurn {type=Curve, loc=<0 0 0>, rot=<0 0 0 1>}
 DisablePelvisAdjustment {type=Curve, loc=<0 0 0>, rot=<0 0 0 1>}
 UncleBrollyDeploy {type=Curve, loc=<0 0 0>, rot=<0 0 0 1>}
 ForceRootSlopeOrientation {type=Curve, loc=<0 0 0>, rot=<0 0 0 1>}
 DisableAdditiveOffsetPose {type=Curve, loc=<0 0 0>, rot=<0 0 0 1>}
 DisableIdleNoiseAdditive {type=Curve, loc=<0 0 0>, rot=<0 0 0 1>}
 UsingConsumable {type=Curve, loc=<0 0 0>, rot=<0 0 0 1>}
 DisableRightArmAdditive {type=Curve, loc=<0 0 0>, rot=<0 0 0 1>}
]
