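# Post-copy anim blueprint for the M_MED_Orbit_Teal body part, handling cloth, gauntlet and loincloth deformation on the parent mesh's pose.
# The AnimGraph runs a deform Control Rig, then nudges gauntlet bones by remapped hand-tracker-to-index-finger socket distances.
# In vehicles, except certain seats like crouch, cannon or shopping cart, it shrinks rear loincloth bones and zeroes cloth blend weight.
# A shared cloth controller function updates the cloth asset each frame, tracking the pelvis.
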AnimBlueprint /BRCosmetics/Characters/Player/Male/Medium/Bodies/M_MED_Orbit_Teal/Meshes/ClothAsset/M_MED_Orbit_Teal_ClothEditor_AnimBP
target_skeleton=/BRCosmetics/Characters/Player/Male/Medium/Bodies/M_MED_Orbit_Teal/Meshes/M_MED_Orbit_Teal_Skeleton
graphs [
 AnimGraph {nodes [
   n1 = Control Rig(Source=Copy Pose From Mesh() {bUseAttachedParent=true}) {
    alphaBoolBlend {blendOption=Linear}
    controlRigAssetReference {blueprintRigClass=&/M_MED_Orbit_Teal_ClothEditor_Deform_CtrlRig_sculpters.M_MED_Orbit_Teal_ClothEditor_Deform_CtrlRig_sculpters_C}
   }
   n2 = "Transform (Modify) Bone - Bone: dyn_gauntlet_l"(
    ComponentPose=Local To Component(LocalPose=@n1)
    Alpha=Get wrist_l_rot() -> (wrist_l_rot:double)
    Translation="-4.5,0,0"
    Rotation="0,0,0"
    Scale="1,1,1"
   ) {
    rotationSpace=BCS_BoneSpace
    translationMode=BMM_Additive
    translationSpace=BCS_BoneSpace
    alphaBoolBlend {blendOption=Linear}
    alphaScaleBiasClamp {bInterpResult=true, interpSpeedDecreasing=20, interpSpeedIncreasing=20}
    boneToModify {boneName=dyn_gauntlet_l}
    translation {x=-4.5}
   }
   n3 = "Transform (Modify) Bone - Bone: dyn_gauntlet_r"(ComponentPose=@n2, Alpha=Get wrist_r_rot() -> (wrist_r_rot:double), Translation="-4.5,0,0", Rotation="0,0,0", Scale="1,1,1") {
    rotationSpace=BCS_BoneSpace
    translationMode=BMM_Additive
    translationSpace=BCS_BoneSpace
    alphaBoolBlend {blendOption=Linear}
    alphaScaleBiasClamp {bInterpResult=true, interpSpeedDecreasing=20, interpSpeedIncreasing=20}
    boneToModify {boneName=dyn_gauntlet_r}
    translation {x=-4.5}
   }
   n4 = Get in vehicle scale alpha() -> (in vehicle scale alpha:double)
   n5 = "Transform (Modify) Bone - Bone: dyn_loincloth_bk_l_02"(ComponentPose=@n3, Alpha=@n4, Translation="0,0,0", Rotation="0,0,0", Scale="0.1,0.5,1") {scaleMode=BMM_Replace, alphaBoolBlend {blendOption=Linear}, boneToModify {boneName=dyn_loincloth_bk_l_02}, scale=<0.1 0.5>}
   n6 = "Transform (Modify) Bone - Bone: dyn_loincloth_bk_02"(ComponentPose=@n5, Alpha=@n4, Translation="0,0,0", Rotation="0,0,0", Scale="0.1,0.5,1") {scaleMode=BMM_Replace, alphaBoolBlend {blendOption=Linear}, boneToModify {boneName=dyn_loincloth_bk_02}, scale=<0.1 0.5>}
   n7 = "Transform (Modify) Bone - Bone: dyn_roincloth_bk_r_02"(ComponentPose=@n6, Alpha=@n4, Translation="0,0,0", Rotation="0,0,0", Scale="0.1,0.5,1") {scaleMode=BMM_Replace, alphaBoolBlend {blendOption=Linear}, boneToModify {boneName=dyn_roincloth_bk_r_02}, scale=<0.1 0.5>}
   Output Pose(Result=Component To Local(ComponentPose=@n7)) {layerGroup=None}
  ]}
 EventGraph {nodes [
   Event BlueprintUpdateAnimation()
   > n1 = Sequence!(execute) -> (then_0, then_1, then_2:Exec, then_3:Exec)
   > Cloth Asset Controller Function!(
    execute
    CharacterPartAnimInstance:Custom Character Part Anim Instance Object Reference=Self-Reference()
    ClothAssetControllerParameters:Fort Cloth Asset Controller State Parameter Modifiers Object Reference=/BRCosmetics/Characters/Player/Common/Cloth/Common_ClothAssetDataParams
    ClothingAssetName:Name=M_MED_Orbit_Teal_ClothEditor_ClothAsset
    TrackedBoneName:Name=pelvis
    JumpFallingDurationSeconds:float=1.3
   ) -> (then, OutAppliedParameters:Cloth Asset Controller Parameters Structure)
   n2 = Get PartSkelMesh() -> (PartSkelMesh:Skeletal Mesh Component Object Reference)
   Set wrist_l_rot(
    execute=@n1.then_1
    wrist_l_rot:double=Get Distance Between Two Sockets!(
     Component:Skeletal Mesh Component Object Reference=@n2
     SocketOrBoneNameA:Name=dyn_hand_loc_tracker_l
     SocketSpaceA:ERelativeTransformSpace Enum=RTS_World
     SocketOrBoneNameB:Name=index_01_l
     SocketSpaceB:ERelativeTransformSpace Enum=RTS_World
     bRemapRange:bool=true
     InRangeMin:float=-5
     InRangeMax:float=9.38
     OutRangeMin:float=1
     OutRangeMax:float=0
    ) -> (ReturnValue:float) {is=title K2Node_CallFunction}
   ) -> (Output_Get:double)
   > Set wrist_r_rot(wrist_r_rot:double=Get Distance Between Two Sockets(
     Component=@n2
     SocketOrBoneNameA=dyn_hand_loc_tracker_r
     SocketSpaceA=RTS_World
     SocketOrBoneNameB=index_01_r
     SocketSpaceB=RTS_World
     bRemapRange=true
     InRangeMin=0
     InRangeMax=9.38
     OutRangeMin=1
     OutRangeMax=0
    )) -> (Output_Get:double)
   n3 = Branch(execute=@n1.then_3, Condition=Get bIsInVehicle() -> (bIsInVehicle:bool))
   n4 = Get bVehicleSeatForcesCrouch() -> (bVehicleSeatForcesCrouch:bool)
   n5 = Get bIsGolfCartVehicleBackPassenger() -> (bIsGolfCartVehicleBackPassenger:bool)
   n6 = Get bIsJackalVehicleDriver() -> (bIsJackalVehicleDriver:bool)
   n7 = Get bIsInShoppingCart() -> (bIsInShoppingCart:bool)
   n8 = Get bIsInCannon() -> (bIsInCannon:bool)
   n9 = Get bIsMeatballPassenger() -> (bIsMeatballPassenger:bool)
   > n10 = Branch(Condition=NOT Boolean(A=OR Boolean!(A=@n4, B=@n5, C:bool=@n6, D:bool=@n7, E:bool=@n8, F:bool=@n9) -> (ReturnValue)))
   > Set in vehicle scale alpha(in vehicle scale alpha:double=1) -> (Output_Get:double)
   n11 = Get PartSkelMesh()
   > Set ClothBlendWeight!(execute, ClothBlendWeight:float=0, self:Skeletal Mesh Component Object Reference=@n11) -> (then, Output_Get:float)
   Set in vehicle scale alpha(execute=[@n10.else, @n3.else], in vehicle scale alpha=0)
   > Set ClothBlendWeight(ClothBlendWeight=1, self=@n11)
   Event BlueprintInitializeAnimation()
   > Set PartSkelMesh(PartSkelMesh:Skeletal Mesh Component Object Reference=GetPartSkeletalMeshComponent()) -> (Output_Get:Skeletal Mesh Component Object Reference)
  ]}
]
variables [
 in vehicle scale alpha {type=number}
 PartSkelMesh {type=SkeletalMeshComponent}
 wrist_l_rot {type=number}
 wrist_r_rot {type=number}
]
