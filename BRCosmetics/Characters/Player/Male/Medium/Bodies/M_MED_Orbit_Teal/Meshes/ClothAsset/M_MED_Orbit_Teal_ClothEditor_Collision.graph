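# Physics asset supplying cloth collision shapes for the M_MED_Orbit_Teal male player body, named as a cloth editor collision asset.
# Two angled capsules on dyn_holster_top form the holster collider, and a wide box on root sits with its top near the origin, like a floor plane.
# Pelvis, spine_01, thigh, calf and foot bodies carry no shapes.
# It has no constraints, so it defines collision volumes without any jointed ragdoll setup.
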PhysicsAsset /BRCosmetics/Characters/Player/Male/Medium/Bodies/M_MED_Orbit_Teal/Meshes/ClothAsset/M_MED_Orbit_Teal_ClothEditor_Collision
bodies {
 calf_l []
 calf_r []
 dyn_holster_top [
  {
   shapeType=Capsule
   length=25.7227
   radius=4.6721
   shapeName=dyn_holster_top_capsule
   center=<16.3968 -0.8675 1.7184>
   rotation=<rot 0 -89.2585 -87.4145>
  }
  {
   shapeType=Capsule
   length=13.5
   radius=4.5
   shapeName=dyn_holster_top_capsule
   center=<4.9486 0.7497 5.9368>
   rotation=<rot 81.8357 -77.3897 179.9999>
  }
 ]
 foot_l []
 foot_r []
 pelvis []
 root [{shapeType=Box, extentX=327.5, extentY=327.5, extentZ=67.8, shapeName=root_box, center=<0 0 -33.4866>}]
 spine_01 []
 thigh_l []
 thigh_r []
}
constraints []
