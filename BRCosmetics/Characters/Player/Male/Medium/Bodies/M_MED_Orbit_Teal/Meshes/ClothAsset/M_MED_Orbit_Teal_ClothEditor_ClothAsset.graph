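# Chaos Cloth asset for the M_MED_Orbit_Teal male medium player body cosmetic, built by an embedded Dataflow graph with two LODs.
# Each LOD merges render section 1 with a sim mesh, adds a backstop accessory mesh, then applies proxy deformer, skinning blend and weight-map-driven simulation configs.
# Capsule physics asset collision and root-bone-relative motion finish each chain.
# LOD1 uses the LOD1 render mesh, density-mapped sim remeshing, PBD stretch/bending and fewer maximum solver iterations.
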ChaosClothAsset /BRCosmetics/Characters/Player/Male/Medium/Bodies/M_MED_Orbit_Teal/Meshes/ClothAsset/M_MED_Orbit_Teal_ClothEditor_ClothAsset
name=EmbeddedDataflow
graph=&self:EmbeddedDataflow
connections [
 {
  fromNode=ProceduralSelection Render
  fromPin=OutputName
  fromPinGuid=#1
  toNode=ProxyDeformer_0
  toPin="SelectionFilterSets[0].RenderSelection.StringValue"
  toPinGuid=#2
 }
 {
  fromNode=AnimDrive WeightMap
  fromPin=OutputName.StringValue
  fromPinGuid=#3
  toNode=SimulationAnimDriveConfig
  toPin=AnimDriveStiffness.WeightMap
  toPinGuid=#4
 }
 {
  fromNode=Backstop Radius WeightMap
  fromPin=OutputName.StringValue
  fromPinGuid=#5
  toNode=SimulationBackstopConfig
  toPin=BackstopRadius.WeightMap
  toPinGuid=#6
 }
 {
  fromNode=SimulationMaxDistanceConfig_0
  fromPin=KinematicVertices3D
  fromPinGuid=#7
  toNode=SkinningBlend
  toPin=KinematicVertices3D.StringValue
  toPinGuid=#8
 }
 {
  fromNode=SimulationGravityConfig
  fromPinGuid=#9
  toNode=SimulationVelocityScaleConfig
  toPinGuid=#10
 }
 {
  fromNode=SimulationMaxDistanceConfig_0
  fromPinGuid=#11
  toNode=SkinningBlend
  toPinGuid=#12
 }
 {
  fromNode=SimAccessoryMeshNode
  fromPinGuid=#13
  toNode=ProceduralSelection Render
  toPinGuid=#14
 }
 {
  fromNode=SkeletalMeshImport_Accessory Mesh
  fromPinGuid=#15
  toNode=SimAccessoryMeshNode
  toPin=SimAccessoryMeshCollection
  toPinGuid=#16
 }
 {
  fromNode=SimulationDampingConfig
  fromPinGuid=#17
  toNode=LiftDrag WeightMap
  toPinGuid=#18
 }
 {
  fromNode=ProceduralSelection SIM
  fromPinGuid=#19
  toNode=ProxyDeformer_0
  toPinGuid=#20
 }
 {
  fromNode=SimulationBackstopConfig
  fromPinGuid=#21
  toNode=SetPhysicsAsset - Capsules_1
  toPinGuid=#22
 }
 {
  fromNode=SimulationLongRangeAttachmentConfig_1
  fromPinGuid=#23
  toNode=SimulationBendingConfig
  toPinGuid=#24
 }
 {
  fromNode=SimulationMassConfig
  fromPinGuid=#25
  toNode=SimulationLongRangeAttachmentConfig_1
  toPinGuid=#26
 }
 {
  fromNode=MergeClothCollections_1
  fromPinGuid=#27
  toNode=SimAccessoryMeshNode
  toPinGuid=#28
 }
 {
  fromNode=SkeletalMeshImport_Render Mesh
  fromPinGuid=#29
  toNode=MergeClothCollections_1
  toPin="Collections[0]"
  toPinGuid=#30
 }
 {
  fromNode=SimulationSolverConfig_0
  fromPinGuid=#31
  toNode=ReferenceBone
  toPinGuid=#32
 }
 {
  fromNode=MaxDist WeightMap
  fromPin=OutputName.StringValue
  fromPinGuid=#33
  toNode=SimulationMaxDistanceConfig_0
  toPin=MaxDistance.WeightMap
  toPinGuid=#34
 }
 {
  fromNode=LiftDrag WeightMap
  fromPin=OutputName.StringValue
  fromPinGuid=#35
  toNode=SimulationAerodynamicsConfig
  toPin=Lift.WeightMap
  toPinGuid=#36
 }
 {
  fromNode=ProceduralSelection SIM
  fromPin=OutputName
  fromPinGuid=#37
  toNode=ProxyDeformer_0
  toPin="SelectionFilterSets[0].SimSelection.StringValue"
  toPinGuid=#38
 }
 {
  fromNode=MaxDist WeightMap
  fromPinGuid=#39
  toNode=SimulationMaxDistanceConfig_0
  toPinGuid=#40
 }
 {
  fromNode=SkeletalMeshImport_Sim Mesh
  fromPinGuid=#41
  toNode=MergeClothCollections_1
  toPin="Collections[1]"
  toPinGuid=#42
 }
 {
  fromNode=ReferenceBone
  fromPinGuid=#43
  toNode=ClothAssetTerminal
  toPin="CollectionLods[0]"
  toPinGuid=#44
 }
 {fromNode=SkinningBlend, fromPinGuid=#45, toNode=WeightMap, toPinGuid=#46}
 {
  fromNode=SimulationCollisionConfig_1
  fromPinGuid=#47
  toNode=SimulationSolverConfig_0
  toPinGuid=#48
 }
 {
  fromNode=ProceduralSelection Render
  fromPinGuid=#49
  toNode=ProceduralSelection SIM
  toPinGuid=#50
 }
 {
  fromNode=Backstop Radius WeightMap
  fromPinGuid=#51
  toNode=SimulationBackstopConfig
  toPinGuid=#52
 }
 {
  fromNode=SimulationStretchConfig
  fromPinGuid=#53
  toNode=SimulationDampingConfig
  toPinGuid=#54
 }
 {fromNode=ProxyDeformer_0, fromPinGuid=#55, toNode=MaxDist WeightMap, toPinGuid=#56}
 {
  fromNode=SimulationAnimDriveConfig
  fromPinGuid=#57
  toNode=Backstop Radius WeightMap
  toPinGuid=#58
 }
 {
  fromNode=SimulationBendingConfig
  fromPinGuid=#59
  toNode=SimulationStretchConfig
  toPinGuid=#60
 }
 {
  fromNode=SetPhysicsAsset - Capsules_1
  fromPinGuid=#61
  toNode=SimulationCollisionConfig_1
  toPinGuid=#62
 }
 {
  fromNode=AnimDrive WeightMap
  fromPinGuid=#63
  toNode=SimulationAnimDriveConfig
  toPinGuid=#64
 }
 {
  fromNode=SkinningBlend
  fromPin=SkinningBlendName
  fromPinGuid=#65
  toNode=WeightMap
  toPin=InputName.StringValue
  toPinGuid=#66
 }
 {
  fromNode=SimulationVelocityScaleConfig
  fromPinGuid=#67
  toNode=AnimDrive WeightMap
  toPinGuid=#68
 }
 {
  fromNode=LiftDrag WeightMap
  fromPinGuid=#69
  toNode=SimulationAerodynamicsConfig
  toPinGuid=#70
 }
 {
  fromNode=LiftDrag WeightMap
  fromPin=OutputName.StringValue
  fromPinGuid=#35
  toNode=SimulationAerodynamicsConfig
  toPin=Drag.WeightMap
  toPinGuid=#71
 }
 {
  fromNode=SimulationAerodynamicsConfig
  fromPinGuid=#72
  toNode=SimulationGravityConfig
  toPinGuid=#73
 }
 {fromNode=WeightMap, fromPinGuid=#74, toNode=SimulationMassConfig, toPinGuid=#75}
 {
  fromNode=Backstop Radius WeightMap_0
  fromPin=OutputName.StringValue
  fromPinGuid=#76
  toNode=SimulationBackstopConfig_0
  toPin=BackstopRadius.WeightMap
  toPinGuid=#77
 }
 {
  fromNode=ReferenceBone_0
  fromPinGuid=#78
  toNode=ClothAssetTerminal
  toPin="CollectionLods[1]"
  toPinGuid=#79
 }
 {
  fromNode=WeightMap Remesh Density_0
  fromPin=OutputName.StringValue
  fromPinGuid=#80
  toNode=Remesh_0
  toPin=DensityMapSim.WeightMap
  toPinGuid=#81
 }
 {
  fromNode=SimulationStretchConfig_0
  fromPinGuid=#82
  toNode=SimulationDampingConfig_0
  toPinGuid=#83
 }
 {
  fromNode=ProceduralSelection Render_0
  fromPinGuid=#84
  toNode=ProceduralSelection SIM_0
  toPinGuid=#85
 }
 {
  fromNode=SkeletalMeshImport_Accessory Mesh_0
  fromPinGuid=#86
  toNode=SimAccessoryMeshNode_0
  toPin=SimAccessoryMeshCollection
  toPinGuid=#87
 }
 {
  fromNode=SimAccessoryMeshNode_0
  fromPinGuid=#88
  toNode=WeightMap Remesh Density_0
  toPinGuid=#89
 }
 {
  fromNode=ProceduralSelection Render_0
  fromPin=OutputName
  fromPinGuid=#90
  toNode=ProxyDeformer_1
  toPin="SelectionFilterSets[0].RenderSelection.StringValue"
  toPinGuid=#91
 }
 {
  fromNode=ProceduralSelection SIM_0
  fromPin=OutputName
  fromPinGuid=#92
  toNode=ProxyDeformer_1
  toPin="SelectionFilterSets[0].SimSelection.StringValue"
  toPinGuid=#93
 }
 {
  fromNode=ProceduralSelection SIM_0
  fromPinGuid=#94
  toNode=ProxyDeformer_1
  toPinGuid=#95
 }
 {
  fromNode=WeightMap Remesh Density_0
  fromPinGuid=#96
  toNode=Remesh_0
  toPinGuid=#97
 }
 {
  fromNode=Remesh_0
  fromPinGuid=#98
  toNode=ProceduralSelection Render_0
  toPinGuid=#99
 }
 {
  fromNode=SimulationMaxDistanceConfig
  fromPinGuid=#100
  toNode=SkinningBlend_0
  toPinGuid=#101
 }
 {
  fromNode=SimulationMaxDistanceConfig
  fromPin=KinematicVertices3D
  fromPinGuid=#102
  toNode=SkinningBlend_0
  toPin=KinematicVertices3D.StringValue
  toPinGuid=#103
 }
 {
  fromNode=SimulationSolverConfig
  fromPinGuid=#104
  toNode=ReferenceBone_0
  toPinGuid=#105
 }
 {
  fromNode=SkeletalMeshImport_Sim Mesh_0
  fromPinGuid=#106
  toNode=MergeClothCollections_0
  toPin="Collections[1]"
  toPinGuid=#107
 }
 {
  fromNode=SkeletalMeshImport_Render Mesh_2
  fromPinGuid=#108
  toNode=MergeClothCollections_0
  toPin="Collections[0]"
  toPinGuid=#109
 }
 {
  fromNode=MergeClothCollections_0
  fromPinGuid=#110
  toNode=SimAccessoryMeshNode_0
  toPinGuid=#111
 }
 {
  fromNode=MaxDist WeightMap_0
  fromPinGuid=#112
  toNode=SimulationMaxDistanceConfig
  toPinGuid=#113
 }
 {
  fromNode=MaxDist WeightMap_0
  fromPin=OutputName.StringValue
  fromPinGuid=#114
  toNode=SimulationMaxDistanceConfig
  toPin=MaxDistance.WeightMap
  toPinGuid=#115
 }
 {
  fromNode=SimulationLongRangeAttachmentConfig
  fromPinGuid=#116
  toNode=SimulationBendingConfig_0
  toPinGuid=#117
 }
 {
  fromNode=SimulationBackstopConfig_0
  fromPinGuid=#118
  toNode=SetPhysicsAsset - Capsules
  toPinGuid=#119
 }
 {
  fromNode=SimulationAnimDriveConfig_0
  fromPinGuid=#120
  toNode=Backstop Radius WeightMap_0
  toPinGuid=#121
 }
 {
  fromNode=AnimDrive WeightMap_0
  fromPinGuid=#122
  toNode=SimulationAnimDriveConfig_0
  toPinGuid=#123
 }
 {
  fromNode=AnimDrive WeightMap_0
  fromPin=OutputName.StringValue
  fromPinGuid=#124
  toNode=SimulationAnimDriveConfig_0
  toPin=AnimDriveStiffness.WeightMap
  toPinGuid=#125
 }
 {
  fromNode=SimulationAerodynamicsConfig_0
  fromPinGuid=#126
  toNode=SimulationGravityConfig_0
  toPinGuid=#127
 }
 {
  fromNode=LiftDrag WeightMap_0
  fromPinGuid=#128
  toNode=SimulationAerodynamicsConfig_0
  toPinGuid=#129
 }
 {
  fromNode=LiftDrag WeightMap_0
  fromPin=OutputName.StringValue
  fromPinGuid=#130
  toNode=SimulationAerodynamicsConfig_0
  toPin=Lift.WeightMap
  toPinGuid=#131
 }
 {
  fromNode=LiftDrag WeightMap_0
  fromPin=OutputName.StringValue
  fromPinGuid=#130
  toNode=SimulationAerodynamicsConfig_0
  toPin=Drag.WeightMap
  toPinGuid=#132
 }
 {
  fromNode=SimulationBendingConfig_0
  fromPinGuid=#133
  toNode=SimulationStretchConfig_0
  toPinGuid=#134
 }
 {
  fromNode=Backstop Radius WeightMap_0
  fromPinGuid=#135
  toNode=SimulationBackstopConfig_0
  toPinGuid=#136
 }
 {
  fromNode=SimulationMassConfig_0
  fromPinGuid=#137
  toNode=SimulationLongRangeAttachmentConfig
  toPinGuid=#138
 }
 {
  fromNode=SimulationDampingConfig_0
  fromPinGuid=#139
  toNode=LiftDrag WeightMap_0
  toPinGuid=#140
 }
 {
  fromNode=SimulationGravityConfig_0
  fromPinGuid=#141
  toNode=SimulationVelocityScaleConfig_0
  toPinGuid=#142
 }
 {
  fromNode=SimulationVelocityScaleConfig_0
  fromPinGuid=#143
  toNode=AnimDrive WeightMap_0
  toPinGuid=#144
 }
 {
  fromNode=SimulationCollisionConfig
  fromPinGuid=#145
  toNode=SimulationSolverConfig
  toPinGuid=#146
 }
 {
  fromNode=SetPhysicsAsset - Capsules
  fromPinGuid=#147
  toNode=SimulationCollisionConfig
  toPinGuid=#148
 }
 {
  fromNode=WeightMap_0
  fromPinGuid=#149
  toNode=SimulationMassConfig_0
  toPinGuid=#150
 }
 {
  fromNode=ProxyDeformer_1
  fromPinGuid=#151
  toNode=MaxDist WeightMap_0
  toPinGuid=#152
 }
 {fromNode=SkinningBlend_0, fromPinGuid=#153, toNode=WeightMap_0, toPinGuid=#154}
 {
  fromNode=SkinningBlend_0
  fromPin=SkinningBlendName
  fromPinGuid=#155
  toNode=WeightMap_0
  toPin=InputName.StringValue
  toPinGuid=#156
 }
]
nodes [
 SimulationMaxDistanceConfig_0 {
  type=FChaosClothAssetSimulationMaxDistanceConfigNode
  guid=#157
  properties {MaxDistance=(bIsAnimatable=True,High=200,WeightMap="MaxDistance")}
 }
 SimulationSolverConfig_0 {
  type=FChaosClothAssetSimulationSolverConfigNode
  guid=#158
  properties {MaxNumIterations=2, NumIterations=1, NumSubstepsImported=(ImportedValue=1)}
 }
 SkeletalMeshImport_Sim Mesh {
  type=FChaosClothAssetSkeletalMeshImportNode_v2
  guid=#159
  properties {
   SkeletalMesh=&SkeletalMesh'&/M_MED_Orbit_Teal_ClothEditor_SIM'
   UVScale=(X=120,Y=120)
   bImportRenderMesh=False
   bImportSimMesh=True
   bImportSingleSection=False
  }
 }
 SkeletalMeshImport_Render Mesh {
  type=FChaosClothAssetSkeletalMeshImportNode_v2
  guid=#160
  properties {
   SectionIndex=1
   SkeletalMesh=&SkeletalMesh'&/M_MED_Orbit_Teal_ClothEditor'
   bImportRenderMesh=True
   bImportSimMesh=False
   bImportSingleSection=True
  }
 }
 MergeClothCollections_1 {type=FChaosClothAssetMergeClothCollectionsNode_v2, guid=#161}
 MaxDist WeightMap {type=FChaosClothAssetWeightMapNode, guid=#162, properties {OutputName=(StringValue="MaxDistance")}}
 SimulationLongRangeAttachmentConfig_1 {
  type=FChaosClothAssetSimulationLongRangeAttachmentConfigNode_v2
  guid=#163
  inputPins [
   Collection {type=FManagedArrayCollection}
   FixedEndSet.StringValue {type=FString}
   TetherStiffness.WeightMap {type=FString}
   TetherScale.WeightMap {type=FString}
   "CustomTetherData[0].CustomFixedEndSet.StringValue" {type=FString}
   "CustomTetherData[0].CustomDynamicEndSet.StringValue" {type=FString}
  ]
  outputPins [Collection {type=FManagedArrayCollection}]
  properties {
   CustomTetherData=(())
   FixedEndSet=(StringValue="KinematicVertices3D")
   TetherScale=(bIsAnimatable=True,Low=1,High=1,WeightMap="TetherScale")
   TetherStiffness=(bIsAnimatable=True,Low=1,High=1,WeightMap="TetherStiffness")
   bEnableCustomTetherGeneration=False
   bUseGeodesicTethers=True
   bWarnDuplicateProperty=True
  }
 }
 ClothAssetTerminal {
  type=FChaosClothAssetTerminalNode_v2
  guid=#164
  inputPins [
   "CollectionLods[0]" {type=FManagedArrayCollection}
   "CollectionLods[1]" {type=FManagedArrayCollection}
  ]
 }
 SimulationBackstopConfig {
  type=FChaosClothAssetSimulationBackstopConfigNode
  guid=#165
  inputPins [
   Collection {type=FManagedArrayCollection}
   BackstopDistance.WeightMap {type=FString}
   BackstopRadius.WeightMap {type=FString}
   BackstopMeshName.StringValue {type=FString}
  ]
  outputPins [Collection {type=FManagedArrayCollection}]
  properties {
   BackstopDistance=(bIsAnimatable=True,WeightMap="BackstopDistance")
   BackstopMeshName=(StringValue="AccessoryMesh")
   BackstopRadius=(bIsAnimatable=True,Low=30,High=100,WeightMap="BackstopRadius")
   bWarnDuplicateProperty=True
  }
 }
 SimulationAnimDriveConfig {
  type=FChaosClothAssetSimulationAnimDriveConfigNode
  guid=#166
  inputPins [
   Collection {type=FManagedArrayCollection}
   AnimDriveStiffness.WeightMap {type=FString}
   AnimDriveDamping.WeightMap {type=FString}
  ]
  outputPins [Collection {type=FManagedArrayCollection}]
  properties {
   AnimDriveDamping=(bIsAnimatable=True,High=1,WeightMap="AnimDriveDamping")
   AnimDriveStiffness=(bIsAnimatable=True,High=1,WeightMap="AnimDriveStiffness")
   bWarnDuplicateProperty=True
  }
 }
 AnimDrive WeightMap {type=FChaosClothAssetWeightMapNode, guid=#167, properties {OutputName=(StringValue="AnimDrive")}}
 SimulationAerodynamicsConfig {
  type=FChaosClothAssetSimulationAerodynamicsConfigNode
  guid=#168
  properties {
   Drag=(bIsAnimatable=True,Low=0.3,High=0.8,WeightMap="Drag")
   Lift=(bIsAnimatable=True,Low=0.2,High=0.4,WeightMap="Lift")
   TurbulenceRatio=1
   WindVelocity=(X=0,Y=0,Z=0)
   WindVelocitySpace=ReferenceBoneSpace
  }
 }
 LiftDrag WeightMap {type=FChaosClothAssetWeightMapNode, guid=#169, properties {OutputName=(StringValue="LiftDrag")}}
 SimulationBendingConfig {
  type=FChaosClothAssetSimulationBendingConfigNode
  guid=#170
  inputPins [
   Collection {type=FManagedArrayCollection}
   FlatnessRatio.WeightMap {type=FString}
   RestAngle.WeightMap {type=FString}
   BendingStiffness.WeightMap {type=FString}
   BendingStiffnessWarp.WeightMap {type=FString}
   BendingStiffnessWeft.WeightMap {type=FString}
   BendingStiffnessBias.WeightMap {type=FString}
   BendingDamping.WeightMap {type=FString}
   BendingAnisoDamping.WeightMap {type=FString}
   AnisoBucklingRatio.WeightMap {type=FString}
   BucklingStiffness.WeightMap {type=FString}
   BucklingStiffnessWarp.WeightMap {type=FString}
   BucklingStiffnessWeft.WeightMap {type=FString}
   BucklingStiffnessBias.WeightMap {type=FString}
   BucklingRatioWeighted.WeightMap {type=FString}
  ]
  properties {
   AnisoBucklingRatio=(bIsAnimatable=True,Low=0.5,High=0.5,WeightMap="BucklingRatio")
   BendingAnisoDamping=(bIsAnimatable=True,Low=0.1,High=5,WeightMap="BendingAnisoDamping")
   BendingStiffnessBias=(bIsAnimatable=True,Low=3,High=5,WeightMap="BendingStiffnessBias")
   BendingStiffnessWarp=(bIsAnimatable=True,Low=3,High=5,WeightMap="BendingStiffnessWarp")
   BendingStiffnessWeft=(bIsAnimatable=True,Low=3,High=5,WeightMap="BendingStiffnessWeft")
   BucklingStiffnessBias=(bIsAnimatable=True,Low=2,High=50,WeightMap="BucklingStiffnessBias")
   BucklingStiffnessWarp=(bIsAnimatable=True,Low=2,High=50,WeightMap="BucklingStiffnessWarp")
   BucklingStiffnessWeft=(bIsAnimatable=True,Low=2,High=50,WeightMap="BucklingStiffnessWeft")
  }
 }
 SimulationStretchConfig {
  type=FChaosClothAssetSimulationStretchConfigNode
  guid=#171
  properties {
   StretchAnisoDamping=(bIsAnimatable=True,Low=2,High=1,WeightMap="StretchAnisoDamping")
   StretchStiffnessBias=(bIsAnimatable=True,Low=10,High=1000000,WeightMap="StretchStiffnessBias")
   StretchStiffnessWarp=(bIsAnimatable=True,Low=100000,High=1000000,WeightMap="StretchStiffnessWarp")
   StretchStiffnessWeft=(bIsAnimatable=True,Low=100000,High=1000000,WeightMap="StretchStiffnessWeft")
  }
 }
 SimulationMassConfig {
  type=FChaosClothAssetSimulationMassConfigNode
  guid=#172
  properties {DensityWeighted=(Low=1,High=1,WeightMap="Density")}
 }
 SimulationDampingConfig {
  type=FChaosClothAssetSimulationDampingConfigNode
  guid=#173
  properties {
   DampingCoefficientWeighted=(WeightMap="DampingCoefficient")
   LocalDampingAngularCoefficient=0.15
   LocalDampingLinearCoefficient=0.15
   LocalDampingSpace=ReferenceBoneSpace
  }
 }
 SimulationGravityConfig {
  type=FChaosClothAssetSimulationGravityConfigNode
  guid=#174
  properties {GravityOverrideImported=(ImportedValue=(X=0,Y=0,Z=-980))}
 }
 SimulationVelocityScaleConfig {
  type=FChaosClothAssetSimulationVelocityScaleConfigNode
  guid=#175
  inputPins [Collection {type=FManagedArrayCollection}]
  outputPins [Collection {type=FManagedArrayCollection}]
  properties {
   AngularVelocityScale=0.75
   FictitiousAngularScale=0.5
   FictitiousForcesModel=FullScale
   LinearVelocityScale=(X=0.75,Y=0.75,Z=0.75)
   MaxAngularAcceleration=9
   MaxAngularVelocity=2
   MaxLinearAcceleration=(X=1400,Y=1000,Z=1400)
   MaxLinearVelocity=(X=700,Y=700,Z=700)
   MaxVelocityScale=2
   VelocityScaleSpace=ReferenceBoneSpace
   bEnableAngularAccelerationClamping=True
   bEnableAngularVelocityClamping=True
   bEnableLinearAccelerationClamping=True
   bEnableLinearVelocityClamping=True
   bWarnDuplicateProperty=True
  }
 }
 SimulationCollisionConfig_1 {
  type=FChaosClothAssetSimulationCollisionConfigNode
  guid=#176
  properties {
   ClothCollisionThicknessImported=(bIsAnimatable=True,WeightMap="ClothCollisionThickness")
   CollisionThicknessImported=(ImportedValue=1)
   FrictionCoefficientWeighted=(bIsAnimatable=True,Low=0.2,High=0.2,WeightMap="FrictionCoefficient")
   InnerCollisionThickness=(WeightMap="InnerCollisionThickness")
   bUseCCD=False
  }
 }
 SetPhysicsAsset - Capsules_1 {
  type=FChaosClothAssetSetPhysicsAssetNode
  guid=#177
  properties {PhysicsAsset=/Script/Engine.PhysicsAsset'&/M_MED_Orbit_Teal_ClothEditor_Collision'}
 }
 WeightMap {
  type=FChaosClothAssetWeightMapNode
  guid=#178
  properties {MapOverrideType=Add, MeshTarget=Render, OutputName=()}
 }
 ProxyDeformer_0 {
  type=FChaosClothAssetProxyDeformerNode_v3
  guid=#179
  properties {InfluenceRadius=5, SelectionFilterSets=(()), bPreserveRenderTangents=True}
 }
 ReferenceBone {
  type=FChaosClothAssetReferenceBoneNode
  guid=#180
  inputPins [Collection {type=FManagedArrayCollection}]
  outputPins [Collection {type=FManagedArrayCollection}]
  properties {CalculateDefaultReferenceBone=(), ReferenceBone=(Name="root")}
 }
 SkinningBlend {
  type=FChaosClothAssetSkinningBlendNode
  guid=#181
  inputPins [
   Collection {type=FManagedArrayCollection}
   KinematicVertices3D.StringValue {type=FString}
  ]
  outputPins [
   Collection {type=FManagedArrayCollection}
   SkinningBlendName {type=FString}
  ]
  properties {KinematicVertices3D=(), SkinningBlendName=RenderDeformerSkinningBlend, bUseSmoothTransition=True}
 }
 SimAccessoryMeshNode {
  type=FChaosClothAssetSimAccessoryMeshNode
  guid=#182
  inputPins [
   Collection {type=FManagedArrayCollection}
   SimAccessoryMeshCollection {type=FManagedArrayCollection}
  ]
  outputPins [
   Collection {type=FManagedArrayCollection}
   AccessoryMeshName {type=FString}
  ]
  properties {AccessoryMeshName=AccessoryMesh, bUseSimImportVertexID=True}
 }
 SkeletalMeshImport_Accessory Mesh {
  type=FChaosClothAssetSkeletalMeshImportNode_v2
  guid=#183
  properties {
   SkeletalMesh=&SkeletalMesh'&/M_MED_Orbit_Teal_ClothEditor_BACKSTOP'
   UVScale=(X=120,Y=120)
   bImportRenderMesh=False
   bImportSimMesh=True
   bImportSingleSection=False
  }
 }
 Backstop Radius WeightMap {type=FChaosClothAssetWeightMapNode, guid=#184, properties {OutputName=(StringValue="BackstopRadiusMap")}}
 ProceduralSelection Render {
  type=FChaosClothAssetProceduralSelectionNode
  guid=#185
  inputPins [
   Collection {type=FManagedArrayCollection}
   ConversionInputName.StringValue {type=FString}
  ]
  outputPins [
   Collection {type=FManagedArrayCollection}
   OutputName {type=FString}
  ]
  properties {ConversionInputName=(), Group=(Name="RenderVertices"), OutputName=RenderSelectionProcedural, SelectionType=SelectAll}
 }
 ProceduralSelection SIM {
  type=FChaosClothAssetProceduralSelectionNode
  guid=#186
  inputPins [
   Collection {type=FManagedArrayCollection}
   ConversionInputName.StringValue {type=FString}
  ]
  outputPins [
   Collection {type=FManagedArrayCollection}
   OutputName {type=FString}
  ]
  properties {ConversionInputName=(), Group=(Name="SimVertices3D"), OutputName=SimSelectionProcedural, SelectionType=SelectAll}
 }
 SimulationMaxDistanceConfig {
  type=FChaosClothAssetSimulationMaxDistanceConfigNode
  guid=#187
  properties {MaxDistance=(bIsAnimatable=True,High=200,WeightMap="MaxDistance")}
 }
 SimulationSolverConfig {
  type=FChaosClothAssetSimulationSolverConfigNode
  guid=#188
  properties {MaxNumIterations=1, NumIterations=1, NumSubstepsImported=(ImportedValue=1)}
 }
 SkeletalMeshImport_Sim Mesh_0 {
  type=FChaosClothAssetSkeletalMeshImportNode_v2
  guid=#189
  properties {
   SkeletalMesh=&SkeletalMesh'&M_MED_Orbit_Teal_ClothEditor_SIM'
   UVScale=(X=120,Y=120)
   bImportRenderMesh=False
   bImportSimMesh=True
   bImportSingleSection=False
  }
 }
 SkeletalMeshImport_Render Mesh_2 {
  type=FChaosClothAssetSkeletalMeshImportNode_v2
  guid=#190
  properties {
   LODIndex=1
   SectionIndex=1
   SkeletalMesh=&SkeletalMesh'&M_MED_Orbit_Teal_ClothEditor'
   bImportRenderMesh=True
   bImportSimMesh=False
   bImportSingleSection=True
  }
 }
 MergeClothCollections_0 {type=FChaosClothAssetMergeClothCollectionsNode_v2, guid=#191}
 MaxDist WeightMap_0 {type=FChaosClothAssetWeightMapNode, guid=#192, properties {OutputName=(StringValue="MaxDistance")}}
 SimulationLongRangeAttachmentConfig {
  type=FChaosClothAssetSimulationLongRangeAttachmentConfigNode_v2
  guid=#193
  inputPins [
   Collection {type=FManagedArrayCollection}
   FixedEndSet.StringValue {type=FString}
   TetherStiffness.WeightMap {type=FString}
   TetherScale.WeightMap {type=FString}
   "CustomTetherData[0].CustomFixedEndSet.StringValue" {type=FString}
   "CustomTetherData[0].CustomDynamicEndSet.StringValue" {type=FString}
  ]
  outputPins [Collection {type=FManagedArrayCollection}]
  properties {
   CustomTetherData=(())
   FixedEndSet=(StringValue="KinematicVertices3D")
   TetherScale=(bIsAnimatable=True,Low=1,High=1,WeightMap="TetherScale")
   TetherStiffness=(bIsAnimatable=True,Low=1,High=1,WeightMap="TetherStiffness")
   bEnableCustomTetherGeneration=False
   bUseGeodesicTethers=True
   bWarnDuplicateProperty=True
  }
 }
 SimulationBackstopConfig_0 {
  type=FChaosClothAssetSimulationBackstopConfigNode
  guid=#194
  inputPins [
   Collection {type=FManagedArrayCollection}
   BackstopDistance.WeightMap {type=FString}
   BackstopRadius.WeightMap {type=FString}
   BackstopMeshName.StringValue {type=FString}
  ]
  outputPins [Collection {type=FManagedArrayCollection}]
  properties {
   BackstopDistance=(bIsAnimatable=True,WeightMap="BackstopDistance")
   BackstopMeshName=(StringValue="AccessoryMesh")
   BackstopRadius=(bIsAnimatable=True,Low=30,High=100,WeightMap="BackstopRadius")
   bWarnDuplicateProperty=True
  }
 }
 SimulationAnimDriveConfig_0 {
  type=FChaosClothAssetSimulationAnimDriveConfigNode
  guid=#195
  inputPins [
   Collection {type=FManagedArrayCollection}
   AnimDriveStiffness.WeightMap {type=FString}
   AnimDriveDamping.WeightMap {type=FString}
  ]
  outputPins [Collection {type=FManagedArrayCollection}]
  properties {
   AnimDriveDamping=(bIsAnimatable=True,High=1,WeightMap="AnimDriveDamping")
   AnimDriveStiffness=(bIsAnimatable=True,High=1,WeightMap="AnimDriveStiffness")
   bWarnDuplicateProperty=True
  }
 }
 AnimDrive WeightMap_0 {type=FChaosClothAssetWeightMapNode, guid=#196, properties {OutputName=(StringValue="AnimDrive")}}
 SimulationAerodynamicsConfig_0 {
  type=FChaosClothAssetSimulationAerodynamicsConfigNode
  guid=#197
  properties {
   Drag=(bIsAnimatable=True,Low=0.3,High=0.8,WeightMap="Drag")
   Lift=(bIsAnimatable=True,Low=0.2,High=0.4,WeightMap="Lift")
   TurbulenceRatio=1
   WindVelocity=(X=0,Y=0,Z=0)
   WindVelocitySpace=ReferenceBoneSpace
  }
 }
 LiftDrag WeightMap_0 {type=FChaosClothAssetWeightMapNode, guid=#198, properties {OutputName=(StringValue="LiftDrag")}}
 SimulationBendingConfig_0 {
  type=FChaosClothAssetSimulationBendingConfigNode
  guid=#199
  inputPins [
   Collection {type=FManagedArrayCollection}
   FlatnessRatio.WeightMap {type=FString}
   RestAngle.WeightMap {type=FString}
   BendingStiffness.WeightMap {type=FString}
   BendingStiffnessWarp.WeightMap {type=FString}
   BendingStiffnessWeft.WeightMap {type=FString}
   BendingStiffnessBias.WeightMap {type=FString}
   BendingDamping.WeightMap {type=FString}
   BendingAnisoDamping.WeightMap {type=FString}
   AnisoBucklingRatio.WeightMap {type=FString}
   BucklingStiffness.WeightMap {type=FString}
   BucklingStiffnessWarp.WeightMap {type=FString}
   BucklingStiffnessWeft.WeightMap {type=FString}
   BucklingStiffnessBias.WeightMap {type=FString}
   BucklingRatioWeighted.WeightMap {type=FString}
  ]
  properties {
   AnisoBucklingRatio=(bIsAnimatable=True,Low=0.5,High=0.5,WeightMap="BucklingRatio")
   BendingAnisoDamping=(bIsAnimatable=True,Low=0.1,High=5,WeightMap="BendingAnisoDamping")
   BendingStiffnessBias=(bIsAnimatable=True,Low=3,High=5,WeightMap="BendingStiffnessBias")
   BendingStiffnessWarp=(bIsAnimatable=True,Low=3,High=5,WeightMap="BendingStiffnessWarp")
   BendingStiffnessWeft=(bIsAnimatable=True,Low=3,High=5,WeightMap="BendingStiffnessWeft")
   BucklingStiffnessBias=(bIsAnimatable=True,Low=2,High=50,WeightMap="BucklingStiffnessBias")
   BucklingStiffnessWarp=(bIsAnimatable=True,Low=2,High=50,WeightMap="BucklingStiffnessWarp")
   BucklingStiffnessWeft=(bIsAnimatable=True,Low=2,High=50,WeightMap="BucklingStiffnessWeft")
   SolverType=PBD
  }
 }
 SimulationStretchConfig_0 {
  type=FChaosClothAssetSimulationStretchConfigNode
  guid=#200
  properties {
   SolverType=PBD
   StretchAnisoDamping=(bIsAnimatable=True,Low=2,High=1,WeightMap="StretchAnisoDamping")
   StretchStiffnessBias=(bIsAnimatable=True,Low=10,High=1000000,WeightMap="StretchStiffnessBias")
   StretchStiffnessWarp=(bIsAnimatable=True,Low=100000,High=1000000,WeightMap="StretchStiffnessWarp")
   StretchStiffnessWeft=(bIsAnimatable=True,Low=100000,High=1000000,WeightMap="StretchStiffnessWeft")
  }
 }
 SimulationMassConfig_0 {
  type=FChaosClothAssetSimulationMassConfigNode
  guid=#201
  properties {DensityWeighted=(Low=1,High=1,WeightMap="Density")}
 }
 SimulationDampingConfig_0 {
  type=FChaosClothAssetSimulationDampingConfigNode
  guid=#202
  properties {
   DampingCoefficientWeighted=(WeightMap="DampingCoefficient")
   LocalDampingAngularCoefficient=0.15
   LocalDampingLinearCoefficient=0.15
   LocalDampingSpace=ReferenceBoneSpace
  }
 }
 SimulationGravityConfig_0 {
  type=FChaosClothAssetSimulationGravityConfigNode
  guid=#203
  properties {GravityOverrideImported=(ImportedValue=(X=0,Y=0,Z=-980))}
 }
 SimulationVelocityScaleConfig_0 {
  type=FChaosClothAssetSimulationVelocityScaleConfigNode
  guid=#204
  inputPins [Collection {type=FManagedArrayCollection}]
  outputPins [Collection {type=FManagedArrayCollection}]
  properties {
   AngularVelocityScale=0.75
   FictitiousAngularScale=0.5
   FictitiousForcesModel=FullScale
   LinearVelocityScale=(X=0.75,Y=0.75,Z=0.75)
   MaxAngularAcceleration=9
   MaxAngularVelocity=2
   MaxLinearAcceleration=(X=1400,Y=1000,Z=1400)
   MaxLinearVelocity=(X=700,Y=700,Z=700)
   MaxVelocityScale=2
   VelocityScaleSpace=ReferenceBoneSpace
   bEnableAngularAccelerationClamping=True
   bEnableAngularVelocityClamping=True
   bEnableLinearAccelerationClamping=True
   bEnableLinearVelocityClamping=True
   bWarnDuplicateProperty=True
  }
 }
 SimulationCollisionConfig {
  type=FChaosClothAssetSimulationCollisionConfigNode
  guid=#205
  properties {
   ClothCollisionThicknessImported=(bIsAnimatable=True,WeightMap="ClothCollisionThickness")
   CollisionThicknessImported=(ImportedValue=1)
   FrictionCoefficientWeighted=(bIsAnimatable=True,Low=0.2,High=0.2,WeightMap="FrictionCoefficient")
   InnerCollisionThickness=(WeightMap="InnerCollisionThickness")
   bUseCCD=False
  }
 }
 SetPhysicsAsset - Capsules {
  type=FChaosClothAssetSetPhysicsAssetNode
  guid=#206
  properties {PhysicsAsset=&PhysicsAsset'&M_MED_Orbit_Teal_ClothEditor_Collision'}
 }
 WeightMap_0 {
  type=FChaosClothAssetWeightMapNode
  guid=#207
  properties {MapOverrideType=Add, MeshTarget=Render, OutputName=()}
 }
 ProxyDeformer_1 {
  type=FChaosClothAssetProxyDeformerNode_v3
  guid=#208
  properties {InfluenceRadius=5, SelectionFilterSets=(()), bPreserveRenderTangents=True}
 }
 ReferenceBone_0 {
  type=FChaosClothAssetReferenceBoneNode
  guid=#209
  inputPins [Collection {type=FManagedArrayCollection}]
  outputPins [Collection {type=FManagedArrayCollection}]
  properties {CalculateDefaultReferenceBone=(), ReferenceBone=(Name="root")}
 }
 SkinningBlend_0 {
  type=FChaosClothAssetSkinningBlendNode
  guid=#210
  inputPins [
   Collection {type=FManagedArrayCollection}
   KinematicVertices3D.StringValue {type=FString}
  ]
  outputPins [
   Collection {type=FManagedArrayCollection}
   SkinningBlendName {type=FString}
  ]
  properties {KinematicVertices3D=(), SkinningBlendName=RenderDeformerSkinningBlend, bUseSmoothTransition=True}
 }
 Backstop Radius WeightMap_0 {type=FChaosClothAssetWeightMapNode, guid=#211, properties {OutputName=(StringValue="BackstopRadiusMap")}}
 ProceduralSelection Render_0 {
  type=FChaosClothAssetProceduralSelectionNode
  guid=#212
  inputPins [
   Collection {type=FManagedArrayCollection}
   ConversionInputName.StringValue {type=FString}
  ]
  outputPins [
   Collection {type=FManagedArrayCollection}
   OutputName {type=FString}
  ]
  properties {ConversionInputName=(), Group=(Name="RenderVertices"), OutputName=RenderSelectionProcedural, SelectionType=SelectAll}
 }
 SimAccessoryMeshNode_0 {
  type=FChaosClothAssetSimAccessoryMeshNode
  guid=#213
  inputPins [
   Collection {type=FManagedArrayCollection}
   SimAccessoryMeshCollection {type=FManagedArrayCollection}
  ]
  outputPins [
   Collection {type=FManagedArrayCollection}
   AccessoryMeshName {type=FString}
  ]
  properties {AccessoryMeshName=AccessoryMesh, bUseSimImportVertexID=True}
 }
 SkeletalMeshImport_Accessory Mesh_0 {
  type=FChaosClothAssetSkeletalMeshImportNode_v2
  guid=#214
  properties {
   SkeletalMesh=&SkeletalMesh'&M_MED_Orbit_Teal_ClothEditor_BACKSTOP'
   UVScale=(X=120,Y=120)
   bImportRenderMesh=False
   bImportSimMesh=True
   bImportSingleSection=False
  }
 }
 ProceduralSelection SIM_0 {
  type=FChaosClothAssetProceduralSelectionNode
  guid=#215
  inputPins [
   Collection {type=FManagedArrayCollection}
   ConversionInputName.StringValue {type=FString}
  ]
  outputPins [
   Collection {type=FManagedArrayCollection}
   OutputName {type=FString}
  ]
  properties {ConversionInputName=(), Group=(Name="SimVertices3D"), OutputName=SimSelectionProcedural, SelectionType=SelectAll}
 }
 Remesh_0 {
  type=FChaosClothAssetRemeshNode_v2
  guid=#216
  inputPins [
   Collection {type=FManagedArrayCollection}
   TriangleSelectionSim.StringValue {type=FString}
   DensityMapSim.WeightMap {type=FString}
   TriangleSelectionRender.StringValue {type=FString}
   DensityMapRender.WeightMap {type=FString}
  ]
  outputPins [Collection {type=FManagedArrayCollection}]
  properties {
   DensityMapRender=(Low=100,High=200,WeightMap="DensityMapRender")
   DensityMapSim=(Low=30,High=100,WeightMap="DensityMapSim")
   DensityMapThresholdRender=1
   DensityMapThresholdSim=0.99
   IterationsRender=10
   IterationsSim=50
   RemeshMethodRender=Remesh
   RenderSeamRemeshIterations=1
   SelectionSourceRender=DensityMapInput
   SelectionSourceSim=DensityMapInput
   SmoothingRender=0.25
   SmoothingSim=0.75
   TargetPercentRender=100
   TriangleSelectionRender=()
   TriangleSelectionSim=()
   bRemeshRender=False
   bRemeshRenderSeams=False
   bRemeshSim=True
  }
 }
 WeightMap Remesh Density_0 {type=FChaosClothAssetWeightMapNode, guid=#217, properties {OutputName=(StringValue="LOD1_RemeshDensity")}}
]
