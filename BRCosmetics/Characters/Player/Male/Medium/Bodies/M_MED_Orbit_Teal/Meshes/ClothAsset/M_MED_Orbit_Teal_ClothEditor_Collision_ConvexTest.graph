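# PhysicsAsset for the M_MED_Orbit_Teal character's cloth, named as a convex collision test.
# The root body holds a wide, flat box whose top sits about level with the origin, forming a floor-like collider.
# Bodies for pelvis, lower spine, legs, feet and a dynamic holster bone list no shapes.
# It has no constraints, so it defines collision only, with no jointed ragdoll.
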PhysicsAsset /BRCosmetics/Characters/Player/Male/Medium/Bodies/M_MED_Orbit_Teal/Meshes/ClothAsset/M_MED_Orbit_Teal_ClothEditor_Collision_ConvexTest
bodies {
 calf_l []
 calf_r []
 dyn_holster_top []
 foot_l []
 foot_r []
 pelvis []
 root [{shapeType=Box, extentX=327.5, extentY=327.5, extentZ=67.8, shapeName=root_box, center=<0 0 -33.4866>}]
 spine_01 []
 thigh_l []
 thigh_r []
}
constraints []
