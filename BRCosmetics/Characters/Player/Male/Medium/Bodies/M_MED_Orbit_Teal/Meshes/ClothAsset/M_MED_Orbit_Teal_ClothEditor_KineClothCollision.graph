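# Physics asset named as kinematic cloth collision for the M_MED_Orbit_Teal cloth editor asset, part of a medium male player body cosmetic.
# It holds a single pelvis body with no collision shapes listed and no constraints.
# As it stands, it appears to be a placeholder that supplies no collision geometry to the cloth simulation.
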PhysicsAsset /BRCosmetics/Characters/Player/Male/Medium/Bodies/M_MED_Orbit_Teal/Meshes/ClothAsset/M_MED_Orbit_Teal_ClothEditor_KineClothCollision
bodies {pelvis []}
constraints []
